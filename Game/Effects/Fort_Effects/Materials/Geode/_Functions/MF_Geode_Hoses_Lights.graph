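# Material function for Geode effects that adds glowing hose liquid and blinking lights to the emissive of incoming material attributes.
# Hose glow pans a ramp, tinted by crack cloud color and masked by red vertex color and a texture; rippled plasma noise only blends identical colors.
# Lights quantize UV1 into cells thresholded against time sine, gating incoming emissive, with optional green vertex mask and no-blink override.
# MPC_Geode drives speeds and intensities.
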MaterialFunction /Game/Effects/Fort_Effects/Materials/Geode/_Functions/MF_Geode_Hoses_Lights
graph {nodes [
  n1 = GetMaterialAttributes(None=FunctionInput() {inputType=FunctionInput_MaterialAttributes}) {attributeGetTypes [#1]}
  n2 = MaterialFunctionCall() {materialFunction=&Swizzle}
  n3 = Multiply(A=Time(), B=CollectionParameter() {collection=/Game/Effects/Fort_Effects/Materials/Geode/MPC_Geode, parameterName=Hose Speed})
  n4 = MaterialFunctionCall!(Use Alternate UVs, Alternate UVs) {materialFunction=/Game/Athena/Effects/Materials/Geode/Functions/MF_Geode_Crack_Colors}
  n5 = TextureCoordinate() {coordinateIndex=1}
  n6 = VectorParameter() {group=Hose, parameterName=Liquid_UV_Scale, sortPriority=5, defaultValue=<rgba 8 2 0 0>}
  n7 = Add(
   A=Multiply(A=@n5, B=ComponentMask(None=@n6.RGB) {g=true, r=true})
   B=ComponentMask(None=TextureSample(UVs=@n5) {samplerType=SAMPLERTYPE_Normal, texture=/Game/Effects/Fort_Effects/Gameplay/Water/Textures/T_WaterRipples_01}.RGB) {g=true, r=true}
  )
  n8 = Multiply(A=Time(), B=ScalarParameter() {defaultValue=2, group=Hose, parameterName=Flow Speed, sortPriority=4})
  n9 = Power(
   Base=TextureSample(UVs=Panner(Coordinate=@n7, Time=@n8) {speedY=-0.5}) {texture=/Game/Effects/Fort_Effects/Textures/Tile/T_Plasma_BallCore}.R
   Exp=ScalarParameter() {defaultValue=1.25, parameterName=Contrast, sortPriority=3}
  )
  n10 = Multiply(
   A=TextureSample(UVs=Panner(Coordinate=@n2.YX, Time=@n3) {speedX=-1}) {texture=/Game/Global/Textures/Default/Masks_and_Gradients/T_Ramp_Virt}.R
   B=LinearInterpolate(A=@n4.Cloud Color, B=@n4.Cloud Color, Alpha=Clamp(None=@n9))
  )
  n11 = CollectionParameter() {collection=&MPC_Geode, parameterName=Hose_Emissive}
  n12 = VertexColor()
  n13 = Multiply(
   A=Multiply(A=Multiply(A=@n10, B=@n11), B=@n12.R)
   B=TextureSample() {texture=/Game/Effects/Fort_Effects/Textures/Geode/T_RocketBits_Hose_M}.RGB
  )
  n14 = TextureCoordinate() {coordinateIndex=1}
  n15 = ScalarParameter() {defaultValue=4, group=Lights, parameterName=Rows/Columns, sortPriority=4}
  n16 = Multiply(A=ComponentMask(None=@n14) {r=true}, B=@n15) {constB=4}
  n17 = Subtract(A=Divide(A=Ceil(None=@n16), B=@n15) {constB=4}) {constB=0.0625}
  n18 = Multiply(A=ComponentMask(None=@n14) {g=true}, B=@n15) {constB=4}
  n19 = Subtract(A=Divide(A=Ceil(None=@n18), B=@n15) {constB=4}) {constB=0.0625}
  n20 = Constant() {r=1}
  n21 = If(
   A=Multiply(A=Clamp(None=@n17), B=Clamp(None=@n19))
   B=MaterialFunctionCall(Speed=CollectionParameter() {collection=&MPC_Geode, parameterName=Blinking Light Speed}) {materialFunction=&DebugTimeSine}.Result
   "A > B"=Constant()
   "A == B"=@n20
   "A < B"=@n20
  )
  n22 = CollectionParameter() {collection=&MPC_Geode, parameterName=Blinking Light Intensity}
  n23 = LinearInterpolate(
   A=Multiply(A=Multiply(A=@n1.EmissiveColor, B=@n21), B=@n22)
   B=Multiply(A=@n1.EmissiveColor, B=@n22)
   Alpha=ScalarParameter() {parameterName=No Blinking}
  )
  n24 = LinearInterpolate(A=Multiply(A=@n12.G, B=@n23), B=@n23, Alpha=ScalarParameter() {group=Lights, parameterName=Use Green Vertex Mask})
  FunctionOutput(None=MaterialFunctionCall!(Material=@n1.MaterialAttributes, Emissive=Add(A=@n13, B=@n24)) {materialFunction=/Engine/Functions/MaterialLayerFunctions/MatLayerBlend_Emissive}.Blended Material)
 ]}
