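# Material function for the Geode effect that layers ice and frost over input material attributes, driven by the MPC_Geode Freeze parameter.
# World-aligned noise, a height gradient and the remapped Freeze value form masks for Ice and Frost attribute sets with blended normals.
# Sparkles crossfade two UV scales chosen from screen derivatives to stay screen-stable, animated by time and Fresnel.
# Several thin-layer Frozen color, roughness and emissive branches are left unconnected.
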MaterialFunction /Game/Effects/Fort_Effects/Materials/Geode/_Functions/MF_Geode_Ice
graph {nodes [
  n1 = TextureCoordinate() {coordinateIndex=1}
  n2 = TextureCoordinate() {coordinateIndex=2}
  n3 = Add(
   A=ComponentMask(None=AppendVector(A=@n1, B=@n2)) {b=true, g=true, r=true}
   B=VectorParameter() {group=Frozen, parameterName=Transition Normals Bias, defaultValue=<rgba 0 0 0.75 0>}.RGB
  ) {desc=bias the normals}
  n4 = ScalarParameter() {defaultValue=3, group=Frozen, parameterName=Transition Noise U}
  n5 = ScalarParameter() {defaultValue=3, group=Frozen, parameterName=Transition Noise V}
  n6 = Add(A=ScalarParameter() {group=Frozen, parameterName=Random}, B=Multiply(A=TextureCoordinate(), B=AppendVector(A=@n4, B=@n5)))
  Add(
   A=DotProduct(A=Saturate(None=@n3), B=Constant3Vector() {constant=<rgba 0 0 1 0>})
   B=Multiply(
    A=TextureSample(UVs=@n6) {texture=/Game/Global/Textures/Default/Tileable_Patterns/T_Smoke_OUT}.G
    B=ScalarParameter() {defaultValue=1, group=Frozen, parameterName=Transition Noise Scale}
   )
  )
  n7 = Saturate(None=NamedRerouteUsage())
  n8 = ScalarParameter() {defaultValue=9, group=Frozen, parameterName=ThinLayerSpread, sortPriority=59}
  n9 = Saturate(None=Power(
    Base=Multiply(A=@n7, B=@n8) {constB=8}
    Exp=ScalarParameter() {defaultValue=2, group=Frozen, parameterName=ThinLayerContrast, sortPriority=59}
   ) {constExponent=12})
  n10 = FunctionInput() {bUsePreviewValueAsDefault=true, description=spec color goes into world position, inputName=MA, inputType=FunctionInput_MaterialAttributes}
  n11 = GetMaterialAttributes(None=NamedRerouteDeclaration(None=@n10) {name=Input MA, nodeColor=<rgba 0.875 0 1 1>}) {attributeGetTypes [
    #1
    #2
    #3
    #4
    #5
    #6
   ]}
  n12 = Multiply(A=Saturate(None=NamedRerouteUsage()), B=ScalarParameter() {defaultValue=9, group=Frozen, parameterName=ThinLayerSpread, sortPriority=59}) {constB=8}
  n13 = ScalarParameter() {defaultValue=2, group=Frozen, parameterName=ThinLayerContrast, sortPriority=59}
  n14 = TextureSampleParameter2D(UVs=Multiply(A=TextureCoordinate(), B=ScalarParameter() {defaultValue=0.85, group=Frozen, parameterName=Ice Normal Tiling})) {group=Frozen, parameterName=Frost Normal, samplerType=SAMPLERTYPE_Normal, texture=/Game/Environments/Sets/Arid/Textures/DesertSand_normal}
  n15 = Saturate(None=NamedRerouteUsage())
  n16 = ScalarParameter() {defaultValue=9, group=Frozen, parameterName=ThinLayerSpread, sortPriority=59}
  n17 = Saturate(None=Power(
    Base=Multiply(A=@n15, B=@n16) {constB=8}
    Exp=ScalarParameter() {defaultValue=2, group=Frozen, parameterName=ThinLayerContrast, sortPriority=59}
   ) {constExponent=12})
  n18 = Normalize(VectorInput=Add(A=@n11.Normal, B=LinearInterpolate(A=@n14.R, Alpha=@n17)))
  n19 = LinearInterpolate(A=@n11.Normal, B=@n18, Alpha=Saturate(None=Power(Base=@n12, Exp=@n13) {constExponent=12}))
  n20 = ScalarParameter() {defaultValue=1, group=Frozen, parameterName=ThinLayerNormalAmount, sortPriority=58}
  n21 = LinearInterpolate(A=@n11.Normal, B=Normalize(VectorInput=LinearInterpolate(A=@n19, B=@n18, Alpha=@n20)))
  n22 = LinearInterpolate(A=NamedRerouteUsage(), B=@n21, Alpha=@n9)
  n23 = ScalarParameter() {defaultValue=0.92, group=Frozen, parameterName=ThinLayerAmount, sortPriority=58}
  n24 = MaterialFunctionCall(
   Brightness=ScalarParameter() {defaultValue=1.4, group=Frozen, parameterName=Emissive Fresnel Brightness}
   Exponent=ScalarParameter() {defaultValue=2, group=Frozen, parameterName=Emissive Fresnel Exponent}
   Vertex / Pixel Normal=Multiply(A=LinearInterpolate(A=@n21, B=@n22, Alpha=@n23), B=VertexNormalWS())
  ) {materialFunction=&MF_Fresnel_FX}
  n25 = Power(Base=Fresnel() {baseReflectFraction=0.005, exponent=4})
  n26 = ScalarParameter() {defaultValue=-6, group=Frozen, parameterName=Edge Frost Fresnel}
  Add(A=@n24.Fresnel, B=LinearInterpolate(B=Multiply(A=@n25, B=@n26)))
  CollectionParameter() {collection=/Game/Effects/Fort_Effects/Materials/Geode/MPC_Geode, parameterName=Freeze Exp}
  n27 = NamedRerouteUsage()
  n28 = NamedRerouteUsage()
  n29 = NamedRerouteUsage()
  n30 = NamedRerouteUsage()
  FunctionOutput(None=BlendMaterialAttributes(A=NamedRerouteUsage(), B=BlendMaterialAttributes(A=@n27, B=@n28, Alpha=@n29), Alpha=Saturate(None=@n30)))
  n31 = LinearInterpolate(
   A=VectorParameter() {group=Frozen, parameterName=Ice Color, defaultValue=<rgba 0.0039 0.0118 0.0431 0>}.RGB
   B=LinearInterpolate(
    A=VectorParameter() {group=Frozen, parameterName=FrozenBaseColorLow, defaultValue=<rgba 0.7529 0.8118 0.8667 0>}.RGB
    B=VectorParameter() {group=Frozen, parameterName=FrozenBaseColorHigh, defaultValue=<rgba 0.8177 0.9359 1 0>}.RGB
    Alpha=Desaturation(None=@n11.BaseColor)
   )
   Alpha=@n9
  )
  LinearInterpolate(
   A=LinearInterpolate(A=@n11.BaseColor, B=@n31)
   B=@n31
   Alpha=ScalarParameter() {defaultValue=0.92, group=Frozen, parameterName=ThinLayerAmount, sortPriority=58}
  )
  n32 = MaterialFunctionCall(Float3=Divide(A=WorldPosition().XYZ, B=Constant() {r=410})) {materialFunction=&BreakOutFloat3Components}
  n33 = TextureObject() {samplerType=SAMPLERTYPE_LinearColor, texture=/Game/Global/Textures/Default/Tileable_Patterns/Cloud_Diffuse}
  n34 = MaterialFunctionCall(X=@n32.R, Y=@n32.R) {materialFunction=&MakeFloat2}
  n35 = ComponentMask(None=VertexNormalWS()) {g=true}
  LinearInterpolate(
   A=TextureSample(
    UVs=MaterialFunctionCall(X=@n32.G, Y=@n32.G) {materialFunction=&MakeFloat2}.Result
    Tex=@n33
   ).RGB
   B=TextureSample(UVs=Add(A=@n34.Result) {constB=0.33}, Tex=@n33).RGB
   Alpha=MaterialFunctionCall(In=Abs(None=@n35), Contrast=Constant() {r=0.2}) {materialFunction=&CheapContrast}.Result
  )
  n36 = Saturate(None=NamedRerouteUsage())
  n37 = ScalarParameter() {defaultValue=9, group=Frozen, parameterName=ThinLayerSpread, sortPriority=59}
  n38 = Saturate(None=Power(
    Base=Multiply(A=@n36, B=@n37) {constB=8}
    Exp=ScalarParameter() {defaultValue=2, group=Frozen, parameterName=ThinLayerContrast, sortPriority=59}
   ) {constExponent=12})
  LinearInterpolate(
   A=ScalarParameter() {defaultValue=0.8, group=Frozen, parameterName=Ice Metallic}
   B=ScalarParameter() {defaultValue=0.3, group=Frozen, parameterName=Frozen Metallic}
   Alpha=@n38
  )
  LinearInterpolate(
   A=ScalarParameter() {defaultValue=0.2, group=Frozen, parameterName=Ice Specular}
   B=ScalarParameter() {defaultValue=0.8, group=Frozen, parameterName=Frozen Specular}
   Alpha=@n38
  )
  LinearInterpolate(
   A=ScalarParameter() {defaultValue=0.05, group=Frozen, parameterName=Ice Roughness}
   B=ScalarParameter() {defaultValue=0.6, group=Frozen, parameterName=Frozen Roughness}
   Alpha=@n38
  )
  n39 = TextureObject() {texture=&T_Smoke_OUT}
  n40 = ObjectPositionWS()
  n41 = WorldPosition()
  n42 = ComponentMask(None=MaterialFunctionCall(TextureObject=@n39, TextureSize=Constant() {r=2000}, WorldPosition=Subtract(A=@n40, B=@n41.XYZ)) {materialFunction=&WorldAlignedTexture}.XY Texture) {r=true}
  n43 = TextureCoordinate()
  n44 = ScalarParameter() {defaultValue=3, group=Frozen, parameterName=UVMaskTiling}
  LinearInterpolate(
   A=@n42
   B=TextureSample(UVs=Multiply(A=@n43, B=@n44), Tex=@n39).R
   Alpha=ScalarParameter() {defaultValue=0.2, group=Frozen, parameterName=WorldToUVMaskBlend}
  )
  n45 = Add(
   A=@n11.EmissiveColor
   B=LinearInterpolate(
    A=VectorParameter() {group=Frozen, parameterName=Frozen AO Color Low Outer, defaultValue=<rgba 0.0706 0.0902 0.1412 0>}.RGB
    B=VectorParameter() {group=Frozen, parameterName=Frozen AO Color Low Inner, defaultValue=<rgba 0.0175 0.0214 0.035 1>}.RGB
    Alpha=@n24.Fresnel
   )
  ) {desc=Emissive}
  Multiply(
   A=Saturate(None=LinearInterpolate(A=@n11.EmissiveColor, B=@n45))
   B=ScalarParameter() {defaultValue=0.1, group=Frozen, parameterName=Emissive Power}
  )
  n46 = Saturate(None=@n42)
  n47 = FunctionInput(Preview=Constant()) {
   bUsePreviewValueAsDefault=true
   description=spec color goes into world position
   inputName=AO Crevis Fill Map
   inputType=FunctionInput_Scalar
  }
  n48 = Power(
   Base=Multiply(A=@n46, B=ScalarParameter() {defaultValue=4, group=Frozen, parameterName=Edge Fill Frost Range, sortPriority=60}) {constB=8}
   Exp=ScalarParameter() {defaultValue=1.65, group=Frozen, parameterName=Edge Fill Frost Contrast, sortPriority=60}
  ) {constExponent=12}
  n49 = Add(
   A=@n46
   B=If(
    A=@n47
    B=ScalarParameter() {defaultValue=0.73, group=Frozen, parameterName=Edge Fill Mask Range, sortPriority=60}
    "A > B"=@n48
    "A == B"=@n48
    "A < B"=Constant()
   )
  )
  n50 = Saturate(None=Multiply(A=@n14.R, B=ScalarParameter() {defaultValue=0.03, parameterName=Thin Layer Normal Color Amount}) {constB=0.01})
  NamedRerouteDeclaration(None=Saturate(None=Add(A=@n49, B=@n50))) {name=LayerMask, nodeColor=<rgba 1 0.375 0 1>}
  n51 = MaterialFunctionCall(
   TextureObject=TextureObject() {texture=&T_Smooth_tile}
   TextureSize=ScalarParameter() {defaultValue=1200, parameterName=MeltMaskScale}
   ProjectionTransitionContrast=Constant() {r=-0.5}
  ) {materialFunction=&WorldAlignedTexture}
  NamedRerouteDeclaration(None=Power(Base=ComponentMask(None=@n51.XY Texture) {g=true}) {constExponent=0.5}) {name=Melt Mask, nodeColor=<rgba 1 0.2109 0 1>}
  n52 = NamedRerouteUsage()
  n53 = NamedRerouteUsage()
  n54 = Max(A=Multiply(A=SmoothStep(Value=@n53) {constMax=0, constMin=0.5}) {constB=0.5}, B=NamedRerouteUsage())
  n55 = NamedRerouteUsage()
  n56 = TextureSampleParameter2D(UVs=Multiply(A=TextureCoordinate(), B=ScalarParameter() {defaultValue=4, group=Frozen, parameterName="Ice  Normal Tiling"})) {
   group=Frozen
   parameterName=Ice Normal
   samplerType=SAMPLERTYPE_Normal
   texture=/Game/Packages/Fortress_SharedMaterials/Textures/DetailTextures/Detail_Paint_N
  }
  n57 = NamedRerouteUsage()
  n58 = MaterialFunctionCall(Normal=@n56.RGB, Flatness=LinearInterpolate(A=Constant() {r=0.8}, Alpha=SmoothStep(Value=@n57) {constMax=0, constMin=10})) {materialFunction=&FlattenNormal}
  n59 = SetMaterialAttributes!(
   MaterialAttributes=NamedRerouteUsage()
   BaseColor=LinearInterpolate(A=Multiply(A=@n52) {constB=0.5}, Alpha=@n54) {desc=Ice Edge}
   Metallic=Constant()
   Specular=Constant() {r=0.224}
   Roughness=LinearInterpolate(Alpha=@n54) {constA=0.15}
   Normal=ShadingPathSwitch(
    Default=MaterialFunctionCall(BaseNormal=@n55, AdditionalNormal=@n58.Result) {materialFunction=&BlendAngleCorrectedNormals}.Result
    Mobile=Add(A=@n56.RGB, B=@n55)
   )
  ) {attributeSetTypes [
    #1
    #4
    #5
    #6
    #3
   ]}
  NamedRerouteDeclaration(None=@n59) {name=Ice MA, nodeColor=<rgba 0 0.4844 1 1>}
  n60 = Constant3Vector() {constant=<rgba 0.3662 0.422 0.5478 1>}
  n61 = Constant3Vector() {constant=<rgba 1 1 1 0>}
  n62 = Fresnel() {baseReflectFraction=0, exponent=3}
  n63 = NamedRerouteUsage()
  n64 = NamedRerouteUsage()
  n65 = VertexNormalWS()
  n66 = Constant() {r=0.1}
  n67 = Constant3Vector() {constant=<rgba 0.2771 0.3229 0.3958 1>}
  n68 = ShadingPathSwitch(Default=Multiply(A=SkyLightEnvMapSample(Direction=@n65, Roughness=@n66)) {constB=0.25}, Forward=@n67, Mobile=@n67)
  n69 = NamedRerouteUsage()
  n70 = NamedRerouteUsage()
  n71 = TextureSampleParameter2D(UVs=Multiply(A=TextureCoordinate(), B=ScalarParameter() {defaultValue=1.5, group=Frozen, parameterName=Frost Normal Tiling})) {group=Frozen, parameterName=Frost Normal, samplerType=SAMPLERTYPE_Normal, texture=&DesertSand_normal}
  n72 = MaterialFunctionCall(
   BaseNormal=@n70
   AdditionalNormal=MaterialFunctionCall(Normal=@n71.RGB, Flatness=Constant() {r=0.5}) {materialFunction=&FlattenNormal}.Result
  ) {materialFunction=&BlendAngleCorrectedNormals}
  n73 = NamedRerouteUsage()
  n74 = SetMaterialAttributes!(
   MaterialAttributes=NamedRerouteUsage()
   BaseColor=LinearInterpolate(A=LinearInterpolate(A=@n60, B=@n61, Alpha=@n62), Alpha=@n63)
   Metallic=Constant()
   Specular=LinearInterpolate(Alpha=@n64) {constA=0.5}
   Roughness=LinearInterpolate(A=LinearInterpolate(Alpha=@n64) {constA=0.5, constB=0.1}, Alpha=@n63)
   EmissiveColor=Add(A=Multiply(A=@n68, B=@n69), B=NamedRerouteUsage())
   Normal=ShadingPathSwitch(
    Default=MaterialFunctionCall(BaseNormal=@n72.Result, AdditionalNormal=@n73) {materialFunction=&BlendAngleCorrectedNormals}.Result
    Mobile=Add(A=@n70, B=@n71.RGB)
   )
  ) {attributeSetTypes [
    #1
    #4
    #5
    #6
    #2
    #3
   ]}
  NamedRerouteDeclaration(None=@n74) {name=Frost MA, nodeColor=<rgba 0.7348 0.9098 1 1>}
  NamedRerouteDeclaration(None=@n11.BaseColor) {name=Input Base, nodeColor=<rgba 0 1 0.3906 1>}
  NamedRerouteDeclaration(None=@n11.EmissiveColor) {name=Input Emissive, nodeColor=<rgba 1 0.0703 0 1>}
  NamedRerouteDeclaration(None=@n11.Normal) {name=Input Normal, nodeColor=<rgba 0 0.25 1 1>}
  n75 = TextureCoordinate()
  n76 = Constant() {r=200}
  n77 = Multiply(A=DDX(Value=@n75), B=@n76)
  n78 = DotProduct(A=@n77, B=@n77)
  n79 = Multiply(A=DDY(Value=@n75), B=@n76)
  n80 = DotProduct(A=@n79, B=@n79)
  n81 = Multiply(A=Logarithm2!(X=Min(A=@n78, B=@n80))) {constB=0.5}
  n82 = Floor(None=@n81)
  n83 = Divide(A=@n75, B=Exponential2!(None=@n82)) {desc=UV A}
  n84 = Divide(A=@n75, B=Exponential2(None=Add(A=@n82))) {desc=UV B}
  n85 = Frac(None=@n81)
  n86 = LinearInterpolate(
   A=TextureSample(UVs=@n83) {samplerType=SAMPLERTYPE_LinearColor, texture=/Game/Environments/Asteria/Terrain/Textures/T_SnowSparkleCircles_M}.RGB
   B=TextureSample(UVs=@n84) {samplerType=SAMPLERTYPE_LinearColor, texture=&T_SnowSparkleCircles_M}.RGB
   Alpha=@n85
  )
  NamedRerouteDeclaration(None=ComponentMask(None=@n86) {g=true}) {name=Ice Edge Noise, nodeColor=<rgba 0.1953 0 1 1>}
  n87 = Time() {bIgnorePause=true}
  NamedRerouteDeclaration(None=ConstantBiasScale(None=Sine(None=@n87) {period=10})) {name=TestTime, nodeColor=<rgba 1 0.1641 0 1>}
  n88 = Constant() {r=410}
  n89 = Constant() {r=0.2}
  n90 = ComponentMask(None=MaterialFunctionCall(
    TextureObject=TextureObject() {texture=&2ndInfectedMapLinear}
    TextureSize=@n88
    ProjectionTransitionContrast=@n89
   ) {materialFunction=&WorldAlignedTexture}.XY Texture) {g=true}
  NamedRerouteDeclaration(None=@n90) {name=Frost Noise Raw, nodeColor=<rgba 0.3828 1 0 1>}
  n91 = Subtract(
   A=VectorParameter() {parameterName=Offset, defaultValue=<rgba 0 0 1600 0>}.RGB
   B=TransformPosition(None=WorldPosition().XYZ) {transformSourceType=TRANSFORMPOSSOURCE_World}
  )
  n92 = Multiply(A=NamedRerouteUsage(), B=Constant() {r=100}) {constB=0.025}
  n93 = Distance(A=ComponentMask(None=Add(A=@n91, B=@n92)) {b=true}, B=ScalarParameter() {defaultValue=150, parameterName=Ice Gradient Fade Distance})
  n94 = ScalarParameter() {defaultValue=800, parameterName=Ice Threshold Distance}
  n95 = OneMinus(None=SmoothStep(Value=Divide(A=@n93, B=@n94)))
  NamedRerouteDeclaration(None=Saturate(None=Multiply(A=@n95) {constB=5})) {name=Ice Gradient Mask, nodeColor=<rgba 1 0 0.7031 1>}
  n96 = TextureObject() {samplerType=SAMPLERTYPE_Normal, texture=/Game/Global/Textures/Default/Tileable_Patterns/2ndInfectedMapNormal}
  n97 = Multiply(A=@n90) {constB=0.3}
  n98 = NamedRerouteUsage()
  n99 = Saturate(None=InverseLinearInterpolate(Value=CollectionParameter() {collection=&MPC_Geode, parameterName=Freeze}) {constA=0.75, desc=Hack to work around Geode sequence starting at 0.75.})
  n100 = Multiply(A=Subtract(A=ComponentMask(None=@n98) {g=true}, B=OneMinus(None=@n99)), B=NamedRerouteUsage()) {constB=10}
  n101 = Multiply(A=Subtract(A=Add(A=@n97, B=@n100)) {constB=0.5}) {constB=4}
  NamedRerouteDeclaration(None=MaterialFunctionCall(
    Normal=MaterialFunctionCall(TextureObject=@n96, TextureSize=@n88, ProjectionTransitionContrast=@n89) {materialFunction=&WorldAlignedNormal}.XY Texture
    Flatness=SmoothStep(Value=@n101) {constMax=1.5, constMin=-0.5}
   ) {materialFunction=&FlattenNormal}.Result) {name=Frost Noise Normal, nodeColor=<rgba 0.487 0.513 1 1>}
  NamedRerouteDeclaration(None=Saturate(None=@n101)) {name=Frost Mask, nodeColor=<rgba 0.7509 0.9694 1 1>}
  n102 = Multiply(A=@n90) {constB=0.1}
  n103 = Multiply(A=Subtract(A=Add(A=@n102, B=@n100) {constB=0}) {constB=0.2}) {constB=50}
  n104 = NamedRerouteUsage()
  NamedRerouteDeclaration(None=ShaderStageSwitch(PixelShader=Add(A=@n103, B=@n104), VertexShader=@n103)) {name=Ice Mask, nodeColor=<rgba 0.1993 0.6472 1 1>}
  n105 = Fresnel() {baseReflectFraction=0, exponent=2}
  n106 = Add(A=Multiply(A=@n105) {constB=10}, B=Multiply(A=Time()) {constB=0.2})
  n107 = TextureSample(UVs=@n83) {samplerType=SAMPLERTYPE_Normal, texture=/Game/Environments/Asteria/Terrain/Textures/T_SnowSparkleCircles_N}
  n108 = SmoothStep(Value=Sine(None=Add(A=@n106, B=@n107.R))) {constMax=0.9, constMin=0.8}
  n109 = TextureSample(UVs=@n84) {samplerType=SAMPLERTYPE_Normal, texture=&T_SnowSparkleCircles_N}
  n110 = SmoothStep(Value=Sine(None=Add(A=@n106, B=@n109.R))) {constMax=0.9, constMin=0.8}
  n111 = Multiply(A=ComponentMask(None=@n86) {r=true}, B=Saturate(None=LinearInterpolate(A=@n108, B=@n110, Alpha=@n85)))
  n112 = LinearInterpolate(Alpha=@n105) {constA=0.01}
  NamedRerouteDeclaration(None=Saturate(None=Multiply(A=@n111, B=@n112))) {name=Frost Sparkle Mask, nodeColor=<rgba 0 1 0.8125 1>}
  n113 = LinearInterpolate(Alpha=@n105) {constA=0.25}
  NamedRerouteDeclaration(None=Saturate(None=Multiply(A=@n111, B=@n113))) {name=Ice Sparkle Mask, nodeColor=<rgba 1 0.4922 0 1>}
  NamedRerouteUsage()
  ScalarParameter() {defaultValue=0.7234, parameterName=TestFreeze, sliderMax=1}
  TextureObject() {texture=/Game/Global/Textures/Default/Tileable_Patterns/T_Thick_Smoke_05}
 ]}
