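# Base Default Lit, Nanite-enabled master material for Geode assets, with static switches for optional features.
# Fortnite_TextureSetLerp combines up to six diffuse, normal and specular-mask sets, whose channels drive specular, metallic and roughness.
# Diffuse and optional emissive, scalable by the MPC_Geode collection, pass through a fresnel function into material attributes.
# Those attributes flow through optional MF_Geode_Hoses_Lights, then MF_Geode_Ice with a crevice AO mask, then MF_CreativeMoveTool.
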Material /Game/Effects/Fort_Effects/Materials/Geode/M_Geode_Base
graph {
 properties {
  bUseMaterialAttributes=true
  bUsedWithNanite=true
  ditheredLODTransition=true
  pixelDepthOffsetMode=PDOM_Legacy
  usedShadingModels=Default Lit
 }
 nodes [
  n1 = TextureSampleParameter2D() {
   group=Specular
   parameterName=SpecularMasks
   samplerSource=SSM_Wrap_WorldGroupSettings
   samplerType=SAMPLERTYPE_Masks
   texture=&T_Fortnite_Default_S
  }
  n2 = TextureSampleParameter2D() {group=Emissive, parameterName=EmissiveTexture, sortPriority=1}
  n3 = StaticComponentMaskParameter(None=AppendVector(A=@n2.RGB, B=@n2.RGB)) {defaultB=true, defaultG=true, defaultR=true, group=Emissive, parameterName=Emissive Mask, sortPriority=3}
  n4 = StaticSwitchParameter(
   True=ComponentMask(None=CollectionParameter() {collection=&/MPC_Geode, parameterName=Lights_EmissiveMultiplier}) {b=true, g=true, r=true}
   False=VectorParameter() {parameterName=Emissive Multiplier, defaultValue=<rgba 1 1 1 1>}.RGB
  ) {defaultValue=true, parameterName=Use MPC}
  n5 = MaterialFunctionCall!(Use UV 2=StaticBoolParameter() {group=Material_1, parameterName=Mat 1 Use UV2_1}) {materialFunction=/Game/Packages/Fortress_SharedMaterials/Base_Material_Functions/UseUV1or2}
  n6 = TextureSampleParameter2D(UVs=@n5.Result) {
   group=Material_1
   parameterName=SpecularMasks
   samplerSource=SSM_Wrap_WorldGroupSettings
   samplerType=SAMPLERTYPE_Masks
   texture=&T_Fortnite_Default_S
  }
  n7 = TextureSampleParameter2D(UVs=@n5.Result) {desc=Use for exterior walls and floors, group=Material_1, parameterName=Diffuse_1, samplerSource=SSM_Wrap_WorldGroupSettings}
  n8 = TextureSampleParameter2D(UVs=@n5.Result) {
   group=Material_1
   parameterName=Normals
   samplerSource=SSM_Wrap_WorldGroupSettings
   samplerType=SAMPLERTYPE_Normal
   texture=/Engine/EngineMaterials/FlatNormal
  }
  n9 = MaterialFunctionCall(Use UV 2=StaticBoolParameter() {group=Multiple_Material_Textures, parameterName=Mat 2 Use UV2}) {materialFunction=&UseUV1or2}
  n10 = TextureSampleParameter2D(UVs=@n9.Result) {
   desc=Use for interior walls and ceilings
   group=Material_2
   parameterName=Normals_Texture_2
   samplerSource=SSM_Wrap_WorldGroupSettings
   samplerType=SAMPLERTYPE_Normal
   texture=&FlatNormal
  }
  n11 = TextureSampleParameter2D(UVs=@n9.Result) {
   desc=Use for interior walls and ceilings
   group=Material_2
   parameterName=Diffuse_Texture_2
   samplerSource=SSM_Wrap_WorldGroupSettings
   texture=/Engine/EngineMaterials/DefaultDiffuse
  }
  n12 = TextureSampleParameter2D(UVs=@n9.Result) {
   desc=Use for interior walls and ceilings
   group=Material_2
   parameterName=SpecularMasks_2
   samplerSource=SSM_Wrap_WorldGroupSettings
   samplerType=SAMPLERTYPE_Masks
   texture=&T_Fortnite_Default_S
  }
  n13 = MaterialFunctionCall(Use UV 2=StaticBoolParameter() {group=Multiple_Material_Textures, parameterName=Mat 3 Use UV2}) {materialFunction=&UseUV1or2}
  n14 = TextureSampleParameter2D(UVs=@n13.Result) {
   desc=Use for trims and accents
   group=Material_3
   parameterName=Diffuse_Texture_3
   samplerSource=SSM_Wrap_WorldGroupSettings
   texture=&DefaultDiffuse
  }
  n15 = TextureSampleParameter2D(UVs=@n13.Result) {
   desc=Use for trims and accents
   group=Material_3
   parameterName=SpecularMasks_3
   samplerSource=SSM_Wrap_WorldGroupSettings
   samplerType=SAMPLERTYPE_Masks
   texture=&T_Fortnite_Default_S
  }
  n16 = TextureSampleParameter2D(UVs=@n13.Result) {
   desc=Use for trims and accents
   group=Material_3
   parameterName=Normals_Texture_3
   samplerSource=SSM_Wrap_WorldGroupSettings
   samplerType=SAMPLERTYPE_Normal
   texture=&FlatNormal
  }
  n17 = MaterialFunctionCall(Use UV 2=StaticBoolParameter() {group=Multiple_Material_Textures, parameterName=Mat 4 Use UV2}) {materialFunction=&UseUV1or2}
  n18 = TextureSampleParameter2D(UVs=@n17.Result) {
   desc=Use for doors
   group=Material_4
   parameterName=Diffuse_Texture_4
   samplerSource=SSM_Wrap_WorldGroupSettings
   texture=&DefaultDiffuse
  }
  n19 = TextureSampleParameter2D(UVs=@n17.Result) {
   desc=Use for doors
   group=Material_4
   parameterName=SpecularMasks_4
   samplerSource=SSM_Wrap_WorldGroupSettings
   samplerType=SAMPLERTYPE_Masks
   texture=&T_Fortnite_Default_S
  }
  n20 = TextureSampleParameter2D(UVs=@n17.Result) {
   desc=Use for doors
   group=Material_4
   parameterName=Normals_Texture_4
   samplerSource=SSM_Wrap_WorldGroupSettings
   samplerType=SAMPLERTYPE_Normal
   texture=&FlatNormal
  }
  n21 = TextureSampleParameter2D() {group=Material_5, parameterName=Diffuse_Texture_5, samplerSource=SSM_Wrap_WorldGroupSettings, texture=&DefaultDiffuse}
  n22 = TextureSampleParameter2D() {
   group=Material_5
   parameterName=Normals_Texture_5
   samplerSource=SSM_Wrap_WorldGroupSettings
   samplerType=SAMPLERTYPE_Normal
   texture=&FlatNormal
  }
  n23 = TextureSampleParameter2D() {
   group=Material_5
   parameterName=SpecularMasks_5
   samplerSource=SSM_Wrap_WorldGroupSettings
   samplerType=SAMPLERTYPE_Masks
   texture=&T_Fortnite_Default_S
  }
  n24 = TextureSampleParameter2D() {group=Material_6, parameterName=Diffuse_Texture_6, samplerSource=SSM_Wrap_WorldGroupSettings, texture=&DefaultDiffuse}
  n25 = TextureSampleParameter2D() {
   group=Material_6
   parameterName=SpecularMasks_6
   samplerSource=SSM_Wrap_WorldGroupSettings
   samplerType=SAMPLERTYPE_Masks
   texture=&T_Fortnite_Default_S
  }
  n26 = TextureSampleParameter2D() {
   group=Material_6
   parameterName=Normals_Texture_6
   samplerSource=SSM_Wrap_WorldGroupSettings
   samplerType=SAMPLERTYPE_Normal
   texture=&FlatNormal
  }
  n27 = MaterialFunctionCall!(
   Spec1=@n6.RGB
   Texture1=AppendVector(A=@n7.RGB, B=@n7.RGB)
   Normal1=@n8.RGB
   Normal2=@n10.RGB
   Texture2=AppendVector(A=@n11.RGB, B=@n11.RGB)
   Spec2=@n12.RGB
   Texture3=AppendVector(A=@n14.RGB, B=@n14.RGB)
   Spec3=@n15.RGB
   Normal3=@n16.RGB
   Texture4=AppendVector(A=@n18.RGB, B=@n18.RGB)
   Spec4=@n19.RGB
   Normal4=@n20.RGB
   Texture5=AppendVector(A=@n21.RGB, B=@n21.RGB)
   Normal5=@n22.RGB
   Spec5=@n23.RGB
   Texture6=AppendVector(A=@n24.RGB, B=@n24.RGB)
   Spec6=@n25.RGB
   Normal6=@n26.RGB
   Texture7
   Texture8
   Texture9
   Use_Dot_Texture9
   Texture10
   Use 2 Diffuse Textures=StaticBoolParameter() {group=Material_2, parameterName=Use 2 Materials}
   --------------
   Use 3 Diffuse Textures=StaticBoolParameter() {group=Material_3, parameterName=Use 3 Materials}
   Use 4 Diffuse Textures=StaticBoolParameter() {group=Material_4, parameterName=Use 4 Materials}
   Use 5 Diffuse Textures=StaticBoolParameter() {group=Material_5, parameterName=Use 5 Materials}
   Use 6 Diffuse Textures=StaticBoolParameter() {group=Material_6, parameterName=Use 6 Materials}
   Use 7 Diffuse Textures
   Use 8 Diffuse Textures
   Use 9 Diffuse Textures
   Use 10 Diffuse Textures
   "-------------- "
   Dot_Texture1
   TransitionInfo
   Dot_Texture2
   Dot_Texture3
   Dot_Texture4
   Dot_Texture5
   Dot_Texture6
   Dot_Texture7
   Dot_Texture8
   Dot_Texture9
   Dot_Texture10
   Use_Dot_Texture1
   Use_Dot_Texture2
   Use_Dot_Texture3
   Use_Dot_Texture4
   Use_Dot_Texture5
   Use_Dot_Texture6
   Use_Dot_Texture7
   Use_Dot_Texture8
   Use_Dot_Texture10
  ) {materialFunction=/Game/Packages/Fortress_SharedMaterials/Base_Material_Functions/Fortnite_TextureSetLerp}
  n28 = MaterialFunctionCall!(
   DiffuseFresnelAmount=Constant() {r=0.2}
   Emissive=StaticSwitchParameter(True=Multiply(A=@n3, B=@n4), False=Constant()) {group=Emissive, parameterName=Use Emissive}
   Base Color=@n27.Diffuse
  ) {desc=diffuse, materialFunction=/Game/Packages/Fortress_SharedMaterials/Base_Material_Functions/FresnelFunction}
  n29 = MakeMaterialAttributes!(
   BaseColor=@n28.d
   Metallic=ComponentMask(None=@n27.SpecularMasks) {desc=metalness, g=true}
   Specular=Multiply(
    A=ScalarParameter() {defaultValue=0.5, group=Specular, parameterName=SpecularMult}
    B=ComponentMask(None=@n27.SpecularMasks) {desc=spec, r=true}
   )
   Roughness=LinearInterpolate(
    A=ScalarParameter() {defaultValue=0.1, group=Specular, parameterName=SpecRoughnessMin}
    B=ScalarParameter() {defaultValue=0.9, group=Specular, parameterName=SpecRoughnessMax}
    Alpha=ComponentMask(None=@n27.SpecularMasks) {b=true, desc=roughness}
   )
   Anisotropy
   EmissiveColor=@n28.d
   Opacity
   OpacityMask
   Normal=@n27.Normals
   Tangent
   WorldPositionOffset
   SubsurfaceColor
   ClearCoat=StaticSwitchParameter(True=@n1.A, False=ScalarParameter() {defaultValue=1, group=Clear Coat, parameterName=Clear Coat Opacity, sliderMax=1}) {defaultValue=true, group=Clear Coat, parameterName=Clear Coat Mask}
   ClearCoatRoughness=ScalarParameter() {group=Clear Coat, parameterName=Clear Coat Roughness}
   AmbientOcclusion
   Refraction
   CustomizedUVs_0
   CustomizedUVs_1
   CustomizedUVs_2
   CustomizedUVs_3
   CustomizedUVs_4
   CustomizedUVs_5
   CustomizedUVs_6
   CustomizedUVs_7
   PixelDepthOffset
   ShadingModel
   Displacement
  )
  n30 = MaterialFunctionCall!(In=@n29) {materialFunction=/Game/Effects/Fort_Effects/Materials/Geode/_Functions/MF_Geode_Hoses_Lights}
  MaterialFunctionCall(MA In=MaterialFunctionCall!(AO Crevis Fill Map=@n1.G, MA=StaticSwitchParameter(True=@n30.Result, False=@n29) {group=Hoses, parameterName=Hoses}) {materialFunction=/Game/Effects/Fort_Effects/Materials/Geode/_Functions/MF_Geode_Ice}.Result) {materialFunction=&MF_CreativeMoveTool}
 ]
}
