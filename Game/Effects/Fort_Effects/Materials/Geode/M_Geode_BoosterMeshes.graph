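# Translucent, unlit, two-sided mesh-particle material for the Geode booster effect, with scrolling dust and turbulence.
# Emissive builds on parallax-offset dust, colored by Geode crack and cloud colors or a plain gradient, then particle color.
# Opacity combines turbulence, plasma edge, ramp and inverse Fresnel, with dynamic parameters for edge, brightness and non-mobile depth fade.
# MF_Geode_Fade masks opacity; both Use Inverse Mask branches take the same inverse output.
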Material /Game/Effects/Fort_Effects/Materials/Geode/M_Geode_BoosterMeshes
graph {
 properties {
  bUsedWithMeshParticles=true
  bUsedWithNiagaraMeshParticles=true
  blendMode=BLEND_Translucent
  pixelDepthOffsetMode=PDOM_Legacy
  shadingModel=MSM_Unlit
  twoSided=true
  usedShadingModels=Unlit
 }
 nodes [
  n1 = TextureCoordinate() {uTiling=2, vTiling=0.5}
  n2 = TextureSample(UVs=BumpOffset(Coordinate=Panner(Coordinate=@n1) {bFractionalPart=true, speedY=-1}, Height=ScalarParameter() {defaultValue=-0.5, parameterName=Height})) {texture=/Game/Global/Textures/Default/PointSplatter/T_Tiling_Dust}
  n3 = TextureSample() {texture=/Game/Global/Textures/Default/Masks_and_Gradients/T_Soft_Ramp_Vert_03}
  n4 = Multiply(A=ScalarParameter() {defaultValue=0.3, parameterName=Gradient Multiplier}, B=@n3.R)
  n5 = MaterialFunctionCall!(Use Alternate UVs, Alternate UVs) {materialFunction=/Game/Athena/Effects/Materials/Geode/Functions/MF_Geode_Crack_Colors}
  n6 = Saturate(None=@n5.Crack Color)
  n7 = Multiply(A=@n2.R, B=Multiply(B=@n6) {constA=5})
  n8 = ComponentMask(None=TextureCoordinate()) {g=true}
  n9 = ScalarParameter() {defaultValue=2}
  n10 = LinearInterpolate(A=@n5.Cloud Color, B=@n6, Alpha=Clamp(None=Power(Base=@n8, Exp=@n9)))
  n11 = ParticleColor()
  n12 = Multiply(A=StaticSwitchParameter(True=Add(A=@n2.R, B=@n4), False=Add(A=@n7, B=@n10)) {parameterName=Use Default Particle Color}, B=@n11.RGB)
  MaterialFunctionCall(Emissive=MaterialFunctionCall(LDR Input=Clamp(None=@n12), HDR Tint=@n12) {materialFunction=&DeriveHDRfromLDR}.RemappedRange) {materialFunction=&MF_FX_EmissiveAdjustments}
  n13 = TextureCoordinate() {vTiling=0.01}
  n14 = TextureCoordinate() {uTiling=4, vTiling=0.025}
  n15 = Add(
   A=TextureSample(UVs=Panner(Coordinate=@n13) {speedX=0.012, speedY=-0.125}) {texture=/Game/Global/Textures/Default/Caustics/T_Turbulence_Seamless}.R
   B=TextureSample(UVs=Panner(Coordinate=@n14) {speedX=0.025, speedY=-0.1}) {texture=&T_Turbulence_Seamless}.R
  )
  n16 = Panner(Coordinate=TextureCoordinate() {uTiling=6}) {speedX=0.125}
  n17 = DynamicParameter() {defaultValue=<rgba 0 0 0 1>, paramNames [LerpEdgeMask,BrightenMask,FadeDistance,Param4]}
  n18 = LinearInterpolate(
   A=@n15
   B=Add(A=TextureSample(UVs=@n16) {texture=/Game/Effects/Fort_Effects/Textures/Tile/T_Plasma_BallEdge}.R, B=@n15)
   Alpha=@n17.LerpEdgeMask
  ) {constAlpha=1}
  n19 = LinearInterpolate(A=Constant(), B=Add(A=Constant() {r=1}, B=@n17.BrightenMask), Alpha=@n3.R)
  n20 = Fresnel(ExponentIn=ScalarParameter() {defaultValue=0.5, parameterName=Fresnel Expo})
  n21 = Multiply(A=@n11.A, B=Multiply(A=Multiply(A=@n18, B=@n19), B=OneMinus(None=@n20)))
  n22 = MaterialFunctionCall!(In) {materialFunction=/Game/Effects/Fort_Effects/Materials/Geode/_Functions/MF_Geode_Fade}
  Multiply(
   A=ShadingPathSwitch(Default=DepthFade(Opacity=@n21, FadeDistance=@n17.FadeDistance), Mobile=@n21) {desc=Migrated from FeatureLevelSwitch}
   B=StaticSwitchParameter(True=@n22.Opacity_Mesh Particle Space_Inverse, False=@n22.Opacity_Mesh Particle Space_Inverse) {parameterName=Use Inverse Mask}
  )
 ]
}
