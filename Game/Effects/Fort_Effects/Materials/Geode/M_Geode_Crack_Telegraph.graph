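# Unlit translucent, refractive mesh-particle material for a Geode crack telegraph effect, combining glass-crack textures, a cubemap function, panning smoke and a gradient-tinted glow.
# A camera-and-normal-offset sphere mask around the particle pivot shapes the glow, tinted by particle color.
# Dynamic parameters set glow offset, radius and Vibrate, which drives crack reveal masking, sine-pulsed glow scale and a world-position-offset function.
# UV gradients, vertex color and deferred-only depth fade shape opacity.
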Material /Game/Effects/Fort_Effects/Materials/Geode/M_Geode_Crack_Telegraph
graph {
 properties {
  bUsedWithMeshParticles=true
  bUsedWithNiagaraMeshParticles=true
  blendMode=BLEND_Translucent
  pixelDepthOffsetMode=PDOM_Legacy
  refractionMethod=RM_IndexOfRefraction
  shadingModel=MSM_Unlit
  usedShadingModels=Unlit
 }
 nodes [
  n1 = TextureCoordinate() {uTiling=3}
  n2 = TextureSample(UVs=@n1) {texture=/Game/Effects/Fort_Effects/Textures/Tile/T_Glass_Cracks_M}
  n3 = DynamicParameter() {defaultValue=<rgba 1 1 0 1>, paramNames [GlowOffset,GlowRadius,Vibrate,Param4]}
  n4 = SphereMask(A=TextureCoordinate() {coordinateIndex=1}, B=Constant() {r=0.5}, Radius=Multiply(A=@n3.Vibrate) {constB=0.5}) {hardnessPercent=90}
  Add(
   A=Multiply(A=LinearInterpolate(Alpha=@n2.R) {constA=0.4, constB=-0.4}, B=Constant3Vector() {constant=<rgba 0 1 0 0>})
   B=LinearInterpolate(
    A=Constant3Vector() {constant=<rgba 0 0 1 0>}
    B=TextureSample(UVs=@n1) {samplerType=SAMPLERTYPE_Normal, texture=/Game/Effects/Fort_Effects/Textures/Tile/T_Glass_Cracks_N}.RGB
    Alpha=@n4
   )
  )
  n5 = ComponentMask(None=TextureCoordinate()) {g=true}
  n6 = Power(Base=@n5) {constExponent=1.4}
  n7 = ParticleColor()
  n8 = MaterialFunctionCall(Float3=Multiply(A=@n2.RGB, B=@n4)) {materialFunction=&BreakOutFloat3Components}
  n9 = Multiply(A=@n8.G) {constB=3}
  n10 = Power(Base=@n5) {constExponent=12}
  n11 = Multiply(A=@n9, B=@n10)
  n12 = VertexColor()
  n13 = OneMinus(None=@n12.B)
  n14 = TextureSample(UVs=Panner(Coordinate=TextureCoordinate() {uTiling=2, vTiling=0.7}) {speedX=0.16, speedY=2}) {texture=/Game/Effects/Fort_Effects/Textures/Tile/T_Inky_Smoke_Tile_Chromatic}
  n15 = Multiply(A=Multiply(A=OneMinus(None=@n12.B), B=Add(A=@n9, B=@n14.RGB))) {constB=0.12}
  n16 = CameraVectorWS()
  n17 = VertexNormalWS()
  n18 = Multiply(
   A=@n3.Vibrate
   B=MaterialFunctionCall!(Time=@n3.Vibrate) {materialFunction=/Game/Effects/Fort_Effects/Materials/Geode/_Functions/MF_SineVibration}.Result
  )
  n19 = Multiply(A=ScalarParameter() {defaultValue=130, parameterName=Scale}, B=Add(A=Constant() {r=50}, B=Multiply(A=@n18) {constB=2}))
  n20 = SphereMask(
   A=Subtract(
    A=WorldPosition().XYZ
    B=MaterialFunctionCall() {materialFunction=&ObjectPivotPoint}.Mesh Particle Pivot Location
   )
   B=Multiply(A=Subtract(A=@n16, B=@n17), B=Multiply(A=@n3.GlowOffset, B=@n19)) {constB=100}
   Radius=Multiply(A=@n3.GlowRadius, B=@n19)
  ) {attenuationRadius=100, hardnessPercent=0}
  n21 = Power(Base=@n20) {constExponent=7}
  n22 = Multiply(A=@n15, B=@n21)
  n23 = Multiply(A=Add(A=Multiply(A=@n11, B=@n13), B=Add(A=@n22, B=@n21))) {constB=5}
  n24 = Add(A=@n15, B=@n20)
  n25 = TextureSample(UVs=Clamp(None=ComponentMask(None=@n24) {g=true, r=true})) {texture=/Game/Effects/Fort_Effects/Textures/Grad/T_Halo_Corona_Gradient}
  Add(
   A=Multiply(A=MaterialFunctionCall!(UV add) {materialFunction=/Game/Athena/Effects/Materials/Geode/Functions/MF_Geode_Cubemap}.Result, B=@n6)
   B=Multiply(A=@n7.RGB, B=Add(A=@n23, B=@n25.RGB))
  )
  n26 = ScalarParameter() {defaultValue=0.05, parameterName=RefractionAdd}
  n27 = Add(A=Multiply(A=@n10, B=Constant() {r=0.15}), B=@n12.G) {constB=0.1}
  n28 = Multiply(A=Multiply(A=@n6, B=LinearInterpolate(Alpha=@n27) {constA=0.22}), B=@n7.A)
  LinearInterpolate(
   B=LinearInterpolate(A=Subtract(B=@n26), B=Add(A=@n26), Alpha=@n8.R) {constA=1.2, constB=1.1}
   Alpha=ShadingPathSwitch(Default=@n28, Deferred=DepthFade(Opacity=@n28) {fadeDistanceDefault=1500})
  ) {constA=1}
  MaterialFunctionCall() {materialFunction=&DebugTimeSine}
  MaterialFunctionCall!(Mask=ComponentMask(None=TextureCoordinate()) {g=true}, Parameter=Add(A=@n3.Vibrate, B=Multiply(A=@n18) {constB=0.05})) {materialFunction=/Game/Effects/Fort_Effects/Materials/Geode/_Functions/MF_Geode_Crack_WPO}
 ]
}
