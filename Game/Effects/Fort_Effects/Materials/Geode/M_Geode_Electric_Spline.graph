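# Translucent, unlit, two-sided particle material for an electric lightning spline in the Geode effects.
# SplineThicken sets strand width, shaped by a ramp and the particle-driven SplineThickness and GrowLightning parameters; GrowLightning also cuts opacity along UV V.
# Noise panned on stepped time, scaled by a gradient texture, adds jagged world-position jitter.
# A scrolling beam texture blends two colors, tinted by particle color, with HDR remapping outside mobile feature levels.
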Material /Game/Effects/Fort_Effects/Materials/Geode/M_Geode_Electric_Spline
graph {
 properties {
  bUsedWithMeshParticles=true
  bUsedWithParticleSprites=true
  blendMode=BLEND_Translucent
  pixelDepthOffsetMode=PDOM_Legacy
  shadingModel=MSM_Unlit
  twoSided=true
  usedShadingModels=Unlit
 }
 nodes [
  n1 = Subtract(
   A=ComponentMask(None=TextureCoordinate()) {g=true}
   B=DynamicParameter() {defaultValue=<rgba 1 425 1 1>, paramNames [GrowLightning, SplineThickness, Param3, Param4]}.GrowLightning
  ) {constB=0.5}
  n2 = OneMinus(None=@n1)
  n3 = TextureCoordinate() {coordinateIndex=1, vTiling=7}
  n4 = Clamp(None=Add(A=TextureSample(UVs=@n3) {texture=/Game/Effects/Fort_Effects/Textures/Ramps/T_Ramp_Smooth_Reverse}.G) {constB=0.6})
  n5 = Multiply(
   A=Multiply(A=Clamp(None=@n2), B=Multiply(A=@n4) {constB=2})
   B=DynamicParameter() {defaultValue=<rgba 1 425 1 1>, paramNames [GrowLightning, SplineThickness, Param3, Param4]}.SplineThickness
  )
  n6 = TextureCoordinate() {uTiling=0.002, vTiling=0.1}
  n7 = ScalarParameter() {defaultValue=0.4, parameterName=Tile}
  n8 = Time()
  n9 = ScalarParameter() {defaultValue=8, parameterName=PanStepSize}
  n10 = Divide(A=Round(None=Multiply(A=@n8, B=@n9)), B=@n9)
  n11 = Constant2Vector() {g=1, r=0.25}
  n12 = ScalarParameter() {defaultValue=-3, parameterName=PanSpeed}
  n13 = TextureSample(UVs=Panner(Coordinate=Multiply(A=@n6, B=@n7), Time=@n10, Speed=Multiply(A=@n11, B=@n12)) {bFractionalPart=true, speedX=0.01, speedY=0.1}) {samplerType=SAMPLERTYPE_LinearColor, texture=/Game/Packages/Fortress_Sky/Materials/Noise32x32}
  n14 = OneMinus(None=@n13.RGB)
  n15 = Panner(
   Coordinate=Multiply(A=TextureCoordinate(), B=Multiply(A=@n7) {constB=6}) {constB=0.1}
   Time=Multiply(A=@n10) {constB=0.5}
   Speed=Multiply(A=@n12) {constB=-1.43}
  ) {bFractionalPart=true, speedX=-0.05}
  n16 = Multiply(
   A=Multiply(A=Subtract(A=@n13.RGB, B=@n14), B=ScalarParameter() {defaultValue=300, parameterName=WPOMult}) {constA=1}
   B=Add(A=TextureSample(UVs=@n15) {texture=/Game/Effects/Fort_Effects/Textures/Grad/T_Vert_Gradient_Sections}.R) {constB=0.15}
  )
  Add(
   A=MaterialFunctionCall(WidthBase=@n5, WidthTip=@n5, UVs for Thickness=TextureCoordinate() {coordinateIndex=1}) {materialFunction=&SplineThicken}.WorldPositionOffset
   B=@n16
  )
  n17 = MaterialFunctionCall() {materialFunction=&Swizzle}
  n18 = Constant2Vector() {g=1, r=3}
  n19 = TextureSample(UVs=Panner(Coordinate=Multiply(A=@n17.YX, B=@n18)) {speedX=-2}) {samplerType=SAMPLERTYPE_LinearColor, texture=/Game/Effects/Fort_Effects/Textures/Tile/T_AMR_Beam_02}
  n20 = TextureSample(UVs=TextureCoordinate() {coordinateIndex=1, vTiling=7}) {texture=&T_Ramp_Smooth_Reverse}
  n21 = ComponentMask(None=TextureCoordinate()) {r=true}
  n22 = Multiply(A=@n21, B=OneMinus(None=@n21))
  n23 = Multiply(A=@n19.R, B=Add(A=@n19.R, B=Multiply(A=@n20.G, B=@n22)))
  n24 = Multiply(
   A=LinearInterpolate(
    A=VectorParameter() {parameterName=Color, defaultValue=<rgba 0 0.4167 0.3979 0>}.RGB
    B=VectorParameter() {parameterName=Color_1, defaultValue=<rgba 1.613 2.4794 3 0>}.RGB
    Alpha=@n23
   )
   B=ParticleColor().RGB
  )
  n25 = MaterialFunctionCall(LDR Input=Clamp(None=@n24), HDR Tint=@n24) {materialFunction=&DeriveHDRfromLDR}
  FeatureLevelSwitch(Default=@n25.RemappedRange, ES2=@n24, ES3_1=@n24, SM4_REMOVED=@n25.RemappedRange)
  LinearInterpolate()
  n26 = OneMinus(None=Ceil(None=@n1))
  Multiply(A=Multiply(A=@n23, B=Clamp(None=@n26)), B=ParticleColor().A)
  ScalarParameter() {defaultValue=425, parameterName=SplineWidth}
  ScalarParameter() {defaultValue=1, parameterName=Lerp, sliderMax=1}
  TransformPosition(None=@n16) {transformType=TRANSFORMPOSSOURCE_World}
 ]
}
