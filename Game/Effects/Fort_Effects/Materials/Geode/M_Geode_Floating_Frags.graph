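# Translucent unlit, two-sided mesh-particle material for floating geode fragments, with cracked-glass refraction.
# Crack textures panned by the Open dynamic parameter drive normals, refraction and crack color, with an optional cubemap layer added to emissive.
# Opacity combines particle alpha, UV gradient and inverted vertex blue, with depth and view-angle fades skipped on low quality.
# Open also scales a UV-weighted vertex offset toward particle center, zeroed on mobile.
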Material /Game/Effects/Fort_Effects/Materials/Geode/M_Geode_Floating_Frags
graph {
 properties {
  bUsedWithMeshParticles=true
  bUsedWithNiagaraMeshParticles=true
  blendMode=BLEND_Translucent
  pixelDepthOffsetMode=PDOM_Legacy
  refractionMethod=RM_IndexOfRefraction
  shadingModel=MSM_Unlit
  twoSided=true
  usedShadingModels=Unlit
 }
 nodes [
  n1 = DynamicParameter() {paramNames [Param1,Param2,Param3,Open]}
  n2 = Panner(Coordinate=TextureCoordinate() {uTiling=2, vTiling=1.2}, Time=@n1.Open) {speedX=0.04, speedY=-0.1}
  n3 = TextureSample(UVs=@n2) {texture=/Game/Effects/Fort_Effects/Textures/Tile/T_Glass_Cracks_M}
  n4 = LinearInterpolate(Alpha=@n3.R) {constA=0.64, constB=-0.64}
  Add(
   A=Multiply(A=@n4, B=Constant3Vector() {constant=<rgba 0 1 0 0>})
   B=MaterialFunctionCall(
    Normal=TextureSample(UVs=@n2) {samplerType=SAMPLERTYPE_Normal, texture=/Game/Effects/Fort_Effects/Textures/Tile/T_Glass_Cracks_N}.RGB
    Flatness=Constant() {r=0.5}
   ) {materialFunction=&FlattenNormal}.Result
  )
  n5 = ScalarParameter() {defaultValue=0.12, parameterName=RefractionAdd}
  n6 = ParticleColor()
  n7 = ComponentMask(None=TextureCoordinate()) {g=true}
  n8 = Multiply(A=@n6.A, B=Power(Base=@n7))
  n9 = ScalarParameter() {defaultValue=100, parameterName=DepthFade}
  n10 = ScalarParameter() {parameterName=AngleFade}
  n11 = QualitySwitch(
   Default=Multiply(
    A=DepthFade(Opacity=@n8, FadeDistance=@n9)
    B=MaterialFunctionCall("FadeSensitivity (optional)"=@n10) {materialFunction=&FadeBasedOnViewAngle}.Result
   )
   Low=@n8
  )
  LinearInterpolate(B=Add(A=Multiply(A=@n4, B=@n5)), Alpha=@n11) {constA=1}
  n12 = VertexColor()
  n13 = Multiply(A=Add(A=LinearInterpolate(Alpha=@n3.G) {constA=0.35, constB=1.2}, B=Multiply(A=@n12.B) {constB=0.5}) {constB=0.5}, B=@n7)
  n14 = Multiply(
   A=Multiply(A=@n6.RGB, B=Power(Base=@n13) {constExponent=3})
   B=MaterialFunctionCall!(Use Alternate UVs, Alternate UVs) {materialFunction=/Game/Athena/Effects/Materials/Geode/Functions/MF_Geode_Crack_Colors}.Crack Color
  )
  n15 = Multiply(
   A=Clamp(None=LinearInterpolate(A=@n6.RGB)) {maxDefault=0.7}
   B=MaterialFunctionCall!(UV add) {materialFunction=/Game/Athena/Effects/Materials/Geode/Functions/MF_Geode_Cubemap}.Result
  )
  MaterialFunctionCall(Emissive=LinearInterpolate(A=@n14, B=Add(A=@n15, B=@n14), Alpha=ScalarParameter() {defaultValue=1, parameterName=UseCubemap})) {materialFunction=&MF_FX_EmissiveAdjustments}
  Multiply(A=OneMinus(None=@n12.B), B=@n11)
  n16 = TextureCoordinate() {coordinateIndex=1}
  n17 = Multiply(A=@n7, B=LinearInterpolate(Alpha=ComponentMask(None=@n16) {g=true}) {constA=0.3})
  n18 = Multiply(A=Subtract(A=ParticlePositionWS(), B=WorldPosition().XYZ)) {constB=1.25}
  ShadingPathSwitch(Default=LinearInterpolate(B=Multiply(A=@n17, B=@n18), Alpha=@n1.Open), Mobile=Constant()) {desc=Migrated from FeatureLevelSwitch}
 ]
}
