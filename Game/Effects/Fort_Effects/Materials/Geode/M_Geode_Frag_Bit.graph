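# Translucent, unlit, two-sided material for geode fragment mesh particles, with a refractive cracked-crystal look.
# Two tiled crack normal and mask layers, the detail one randomized per particle, drive Fresnel, refraction and particle-alpha-scaled opacity.
# Emissive combines geode cubemap and crack-color functions with particle color.
# GlitchWPO, off by default, adds world-masked glitch flashes and stepped offsets; an alternate vertex-color rotation and noise wobble is also switched off by default.
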Material /Game/Effects/Fort_Effects/Materials/Geode/M_Geode_Frag_Bit
graph {
 properties {
  bUsedWithMeshParticles=true
  bUsedWithNiagaraMeshParticles=true
  blendMode=BLEND_Translucent
  pixelDepthOffsetMode=PDOM_Legacy
  shadingModel=MSM_Unlit
  twoSided=true
  usedShadingModels=Unlit
 }
 nodes [
  n1 = TextureCoordinate()
  n2 = VectorParameter() {desc="Crack tiling Main(r,g) Detail(b,a)", parameterName=Crack Tiling, defaultValue=<rgba 1 1 3.5 4.2>}
  n3 = Multiply(A=@n1, B=MaterialFunctionCall(X=@n2.R, Y=@n2.R) {materialFunction=&MakeFloat2}.Result)
  n4 = Add(
   A=Multiply(A=@n1, B=MaterialFunctionCall(X=@n2.B, Y=@n2.B) {materialFunction=&MakeFloat2}.Result)
   B=ParticleRandom()
  )
  n5 = Add(
   A=TextureSample(UVs=@n3) {samplerType=SAMPLERTYPE_Normal, texture=/Game/Effects/Fort_Effects/Textures/Tile/T_Glass_Cracks_N}.RGB
   B=Multiply(
    A=TextureSample(UVs=@n4) {samplerType=SAMPLERTYPE_Normal, texture=&T_Glass_Cracks_N}.RGB
    B=Constant3Vector() {constant=<rgba 1 1 0 0>}
   )
  )
  n6 = Fresnel(
   ExponentIn=ScalarParameter() {defaultValue=4, parameterName=FresnelExp}
   BaseReflectFractionIn=ScalarParameter() {parameterName=FresnelBaseRefract}
   Normal=Multiply(A=Transform(None=@n5), B=TwoSidedSign())
  ) {baseReflectFraction=0, exponent=4}
  n7 = TextureSample(UVs=@n3) {texture=/Game/Effects/Fort_Effects/Textures/Tile/T_Glass_Cracks_M}
  n8 = TextureSample(UVs=@n4) {texture=&T_Glass_Cracks_M}
  n9 = Add(A=@n7.G, B=Multiply(A=@n8.G) {constB=0.1})
  n10 = LinearInterpolate(A=ScalarParameter() {defaultValue=0.25, parameterName=MinOpacity}, Alpha=@n9) {constA=0.25}
  n11 = Add(A=@n6, B=@n10)
  n12 = ParticleColor()
  LinearInterpolate(
   B=ScalarParameter() {defaultValue=1.1, parameterName=Refraction}
   Alpha=Multiply(A=ScalarParameter() {defaultValue=1, parameterName=Opacity}, B=Multiply(A=@n11, B=@n12.A))
  ) {constA=1}
  n13 = LinearInterpolate(
   A=Multiply(A=VectorParameter() {parameterName=Glitch2, defaultValue=<rgba 31.2957 4.7459 100 0>}.RGB, B=@n7.R)
   B=Multiply(A=VectorParameter() {parameterName=Glitch1, defaultValue=<rgba 0 100 60.7947 0>}.RGB, B=Add(A=@n7.G, B=@n8.G))
  )
  n14 = Constant() {r=-2}
  n15 = DynamicParameter() {paramNames [GlitchWPOMult, GlitchMoveSpeed, GlitchWPOSwitch, Param4]}
  n16 = Multiply(A=Time(), B=@n15.GlitchMoveSpeed)
  n17 = ConstantBiasScale(None=Sine(None=@n16) {period=0.117})
  n18 = MaterialFunctionCall!(MaskScaleMult=Constant() {r=4}) {materialFunction=/Game/Athena/Effects/Materials/Meteor/MF_Glitch_FX_World_Mask}
  n19 = Multiply(
   A=DynamicParameter() {paramNames [GlitchWPOMult, GlitchMoveSpeed, GlitchWPOSwitch, Param4]}.GlitchWPOSwitch
   B=Multiply(A=Desaturation(None=@n13, Fraction=@n14), B=Multiply(A=@n17, B=@n18.WorldSpaceMask))
  )
  n20 = Constant()
  n21 = MaterialFunctionCall!(UV add) {materialFunction=/Game/Athena/Effects/Materials/Geode/Functions/MF_Geode_Cubemap}
  n22 = MaterialFunctionCall!(Use Alternate UVs, Alternate UVs) {materialFunction=/Game/Athena/Effects/Materials/Geode/Functions/MF_Geode_Crack_Colors}
  n23 = Desaturation(None=Multiply(A=@n6, B=@n10), Fraction=ScalarParameter() {parameterName=Geode Cloud Color Desaturation})
  n24 = Multiply(A=Desaturation(None=@n22.Crack Color, Fraction=ScalarParameter() {parameterName=Geode Crack Color Desaturation})) {constB=8}
  n25 = Multiply(
   A=Add(A=Multiply(A=@n23, B=@n22.Cloud Color), B=Multiply(A=@n24, B=@n9))
   B=Multiply(A=ScalarParameter() {defaultValue=1, parameterName=Strength}, B=@n12.RGB)
  )
  MaterialFunctionCall(Emissive=Add(A=StaticSwitchParameter(True=@n19, False=@n20) {parameterName=GlitchWPO}, B=Add(A=@n21.Result, B=@n25))) {materialFunction=&MF_FX_EmissiveAdjustments}
  n26 = Multiply(
   A=VectorParameter() {parameterName=MeteorDirection, defaultValue=<rgba 0.7755 -0.4565 -0.4361 0>}.RGB
   B=ScalarParameter() {defaultValue=100, parameterName=GlitchWPODist}
  )
  n27 = Round(None=Multiply(A=Sine(None=@n16) {period=0.11}, B=Multiply(A=@n26, B=@n15.GlitchWPOMult)))
  n28 = Time()
  n29 = VertexColor()
  n30 = Multiply(
   A=RotateAboutAxis(
    NormalizedRotationAxis=Constant3Vector() {constant=<rgba 0.25 0.3 1 0>}
    RotationAngle=Multiply(A=@n28) {constB=0.013}
    PivotPoint=ObjectPositionWS()
    Position=WorldPosition().XYZ
   )
   B=@n29.R
  )
  n31 = Time()
  n32 = Multiply(
   A=@n29.G
   B=RotateAboutAxis(
    NormalizedRotationAxis=Constant3Vector() {constant=<rgba 0.5 0 1 0>}
    RotationAngle=Multiply(A=@n31) {constB=-0.02}
    PivotPoint=ObjectPositionWS()
    Position=WorldPosition().XYZ
   )
  )
  n33 = RotateAboutAxis(
   NormalizedRotationAxis=Constant3Vector() {constant=<rgba 0 0.5 1 0>}
   RotationAngle=Multiply(A=Time()) {constB=0.028}
   PivotPoint=ObjectPositionWS()
   Position=WorldPosition().XYZ
  )
  n34 = Panner(Coordinate=TextureCoordinate() {coordinateIndex=1}) {speedX=0.15, speedY=0.1}
  n35 = Add(
   A=Add(A=Add(A=@n30, B=@n32), B=Multiply(A=@n29.B, B=@n33))
   B=Multiply(
    A=TextureSample(UVs=@n34) {samplerType=SAMPLERTYPE_Normal, texture=&T_Weap_Distort_N_02}.RGB
    B=ScalarParameter() {defaultValue=8.5, parameterName=Param_2}
   )
  )
  n36 = Subtract(A=Subtract(A=WorldPosition().XYZ, B=ObjectPositionWS()), B=Time())
  n37 = ScalarParameter() {defaultValue=25}
  n38 = Multiply(
   A=VectorNoise(World Position=Divide(A=@n36, B=@n37)) {noiseFunction=VNF_VectorALU}
   B=ScalarParameter() {defaultValue=13, parameterName=Param_1}
  )
  StaticSwitchParameter(
   True=Multiply(A=LinearInterpolate(Alpha=@n18.DistanceFadeMask) {constB=3}, B=Multiply(A=@n27, B=@n18.WorldSpaceMask))
   False=StaticSwitchParameter(True=Constant(), False=Add(A=@n35, B=@n38)) {defaultValue=true, parameterName=World Position Offset}
  ) {parameterName=GlitchWPO}
 ]
}
