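# Translucent, surface-lit glass material for the Geode lights effect, usable on skeletal meshes and mesh particles.
# A tight Fresnel rim multiplies a gradient texture, defaulting to a DragonflyWings wrap gradient.
# The Switch_Emissive_Color scalar blends that rim from black before it feeds MF_FX_EmissiveAdjustments.
# Parameters cover metallic, roughness, specular, opacity, refraction, tube color and clear coat color, though the shading model is Default Lit.
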Material /Game/Effects/Fort_Effects/Materials/Geode/M_Geode_Lights_Glass
graph {
 properties {
  bUsedWithMeshDeformer=true
  bUsedWithMeshParticles=true
  bUsedWithNiagaraMeshParticles=true
  bUsedWithSkeletalMesh=true
  blendMode=BLEND_Translucent
  pixelDepthOffsetMode=PDOM_Legacy
  translucencyLightingMode=TLM_Surface
  usedShadingModels=Default Lit
 }
 nodes [
  Constant3Vector() {constant=<rgba 0 0 1 0>}
  Constant()
  n1 = Fresnel() {baseReflectFraction=0, exponent=6}
  n2 = TextureSampleParameter2D() {parameterName=EmissiveColor_Texture, texture=/Game/Weapons/WeaponSkins/Wraps/Textures/DragonflyWings/T_Wrap_DragonflyWings_Grad}
  MaterialFunctionCall(Emissive=LinearInterpolate(A=Constant(), B=Multiply(A=@n1, B=@n2.RGB), Alpha=ScalarParameter() {parameterName=Switch_Emissive_Color})) {materialFunction=&MF_FX_EmissiveAdjustments}
  ScalarParameter() {parameterName=Specular}
  ScalarParameter() {defaultValue=0.25, parameterName=Roughness}
  ScalarParameter() {defaultValue=1, parameterName=Metallic}
  ScalarParameter() {defaultValue=0.35, parameterName=Opacity}
  ScalarParameter() {defaultValue=1.025, parameterName=refraction}
  VectorParameter() {parameterName=Tube Color, defaultValue=<rgba 0.5 0.4 0.2 1>}
  VectorParameter() {parameterName=Clear_Coat_Color, defaultValue=<rgba 1 1 1 1>}
 ]
}
