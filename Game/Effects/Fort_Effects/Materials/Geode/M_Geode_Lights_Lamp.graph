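# Default Lit material for a Geode effect lamp light, enabled for mesh particles and built on the AB_Research siren light textures.
# A gradient panning over swizzled UVs pulses the siren emissive mask between dim and full strength.
# The clamped result is multiplied by particle color and an inverted Fresnel, then feeds MF_FX_EmissiveAdjustments.
# It also contains a sphere mask branch, siren D, N and S textures, and scalar surface parameters, with no consumers shown.
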Material /Game/Effects/Fort_Effects/Materials/Geode/M_Geode_Lights_Lamp
graph {
 properties {bUsedWithMeshParticles=true, pixelDepthOffsetMode=PDOM_Legacy, usedShadingModels=Default Lit}
 nodes [
  n1 = MaterialFunctionCall(XY=TextureCoordinate()) {materialFunction=&Swizzle}
  n2 = LinearInterpolate(
   A=Constant() {r=0.35}
   B=Constant() {r=1}
   Alpha=TextureSample(UVs=Panner(Coordinate=@n1.YX) {speedY=2}) {samplerType=SAMPLERTYPE_LinearColor, texture=/Game/Global/Textures/Default/Masks_and_Gradients/T_VertGradient}.G
  )
  Add(
   A=@n2
   B=SphereMask(
    A=TextureCoordinate() {uTiling=0}
    B=Constant2Vector() {g=0.02}
    Radius=ScalarParameter() {defaultValue=0.025, parameterName=Radius}
    Hardness=ScalarParameter() {defaultValue=0.8, parameterName=Hardness}
   )
  )
  n3 = Fresnel() {exponent=0.25}
  n4 = TextureSample() {texture=/Game/Environments/Sets/AB_Research/Textures/T_Siren_Light_E}
  n5 = Multiply(A=Multiply(A=ParticleColor().RGB, B=OneMinus(None=@n3)), B=Clamp(None=Multiply(A=@n4.R, B=@n2)))
  LinearInterpolate(A=@n5, B=Constant())
  MaterialFunctionCall(Emissive=@n5) {materialFunction=&MF_FX_EmissiveAdjustments}
  ScalarParameter() {parameterName=Metallic}
  ScalarParameter() {parameterName=Specular}
  ScalarParameter() {parameterName=Roughness}
  TextureSample() {texture=/Game/Environments/Sets/AB_Research/Textures/T_Siren_Light_D}
  TextureSample() {samplerType=SAMPLERTYPE_Normal, texture=/Game/Environments/Sets/AB_Research/Textures/T_Siren_Light_N}
  TextureSample() {texture=/Game/Environments/Sets/AB_Research/Textures/T_Siren_Light_S}
 ]
}
