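# Translucent unlit mesh-particle material for the Geode radial warp effect, drawing glowing cracks that animate open and closed.
# The dynamic Open parameter slides glass-crack textures, thresholds their masks and, with sine vibration, drives optional world position offset.
# Geode cubemap and crack-color functions build the emissive, tinted by particle color and crack normals (skipped on low quality).
# Crack masks and vertex red shape depth-faded opacity, with a smoke-based erode animation.
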Material /Game/Effects/Fort_Effects/Materials/Geode/M_Geode_Radial_Warp
graph {
 properties {
  bUsedWithMeshParticles=true
  bUsedWithNiagaraMeshParticles=true
  blendMode=BLEND_Translucent
  pixelDepthOffsetMode=PDOM_Legacy
  shadingModel=MSM_Unlit
  usedShadingModels=Unlit
 }
 nodes [
  n1 = VectorParameter() {parameterName=Tile, defaultValue=<rgba 1 1 1 1>}
  n2 = Multiply(A=TextureCoordinate() {uTiling=2, vTiling=1.25}, B=AppendVector(A=@n1.R, B=@n1.R))
  n3 = DynamicParameter() {paramNames [Param1,Param2,Param3,Open]}
  n4 = AppendVector(
   A=Constant()
   B=LinearInterpolate(
    A=ScalarParameter() {defaultValue=-0.73, parameterName=OpenMin}
    B=ScalarParameter() {defaultValue=-0.45, parameterName=OpenMax}
    Alpha=@n3.Open
   ) {constA=-0.73, constB=-0.45}
  )
  n5 = Add(A=@n2, B=@n4)
  n6 = TextureSample(UVs=@n5) {samplerType=SAMPLERTYPE_Normal, texture=/Game/Effects/Fort_Effects/Textures/Tile/T_Glass_Cracks_N}
  n7 = TextureSample(UVs=@n5) {texture=/Game/Effects/Fort_Effects/Textures/Tile/T_Glass_Cracks_M}
  Add(A=@n6.RGB, B=Multiply(A=LinearInterpolate(Alpha=@n7.R) {constA=0.5, constB=-0.5}, B=Constant3Vector() {constant=<rgba 0 1 0 0>}))
  n8 = ParticleColor()
  n9 = VertexColor()
  n10 = Power(Base=@n9.R) {constExponent=3}
  n11 = LinearInterpolate(Alpha=@n3.Open) {constA=-0.1, constB=0.5}
  n12 = SmoothStep(Min=@n11, Max=Add(A=@n11) {constB=0.02}, Value=@n7.R)
  n13 = AppendVector(A=@n1.B, B=@n1.B)
  n14 = TextureSampleParameter2D(UVs=Add(A=Multiply(A=@n2, B=@n13), B=@n4)) {parameterName=CracksMask, texture=/Game/Effects/Fort_Effects/Textures/Tile/T_Glass_Cracks_Mask}
  n15 = StaticSwitchParameter(True=Saturate(None=Add(A=@n12, B=@n14.G)), False=@n14.G) {defaultValue=true, parameterName=NewCloseAnim}
  n16 = Multiply(A=@n10, B=@n15)
  n17 = Multiply(A=Power(Base=OneMinus(None=@n12)) {constExponent=5}) {constB=0.5}
  n18 = NamedRerouteUsage()
  n19 = StaticSwitchParameter(True=Add(A=LinearInterpolate(A=@n17, Alpha=@n14.G) {constB=0}, B=Multiply(A=@n18, B=@n7.G)), False=@n7.G) {defaultValue=true, parameterName=NewCloseAnim}
  n20 = LinearInterpolate(Alpha=@n19) {constA=0.8, constB=2}
  DepthFade(Opacity=Multiply(A=@n8.A, B=Multiply(A=@n16, B=@n20)), FadeDistance=ScalarParameter() {parameterName=DepthFade})
  MaterialFunctionCall() {materialFunction=&DebugTimeSine}
  n21 = MaterialFunctionCall!(UV add) {materialFunction=/Game/Athena/Effects/Materials/Geode/Functions/MF_Geode_Cubemap}
  n22 = MaterialFunctionCall!(Use Alternate UVs, Alternate UVs) {materialFunction=/Game/Athena/Effects/Materials/Geode/Functions/MF_Geode_Crack_Colors}
  n23 = Multiply(A=@n22.Crack Color) {constB=25}
  n24 = Constant3Vector() {constant=<rgba 0 1 0 0>}
  n25 = Saturate(None=Multiply(A=DotProduct(A=@n6.RGB, B=@n24)) {constB=5})
  n26 = Add(A=OneMinus(None=@n15), B=QualitySwitch(Default=Multiply(A=@n10, B=@n25), Low=Constant()))
  n27 = Multiply(
   A=Multiply(A=@n19, B=ScalarParameter() {defaultValue=2, parameterName=CrackIntensity}) {constB=2}
   B=Add(A=@n22.Cloud Color, B=Multiply(A=@n23, B=@n26) {constB=2})
  )
  MaterialFunctionCall(Emissive=Add(A=Multiply(A=@n21.Result, B=@n9.R), B=Multiply(A=@n8.RGB, B=@n27))) {materialFunction=&MF_FX_EmissiveAdjustments}
  MaterialFunctionCall() {materialFunction=&DebugTimeSine}
  MaterialFunctionCall(Speed=Constant() {r=0.5}) {materialFunction=&DebugTimeSine}
  n28 = DynamicParameter() {parameterIndex=1, defaultValue=<rgba 0.11 1 1 1>, paramNames [Erode,Param2,Param3,Param4]}
  n29 = SmoothStep(
   Min=@n28.Erode
   Max=Add(A=@n28.Erode) {constB=0.1}
   Value=Add(A=TextureSample() {texture=/Game/Effects/Fort_Effects/Textures/Tile/T_Inky_Smoke_Tile_Chromatic}.RGB) {constB=0.17}
  )
  NamedRerouteDeclaration(None=Desaturation(None=Saturate(None=@n29), Fraction=Constant() {r=0.4})) {name=ErodeAnim, nodeColor=<rgba 1 0 0.3281 1>}
  n30 = TextureCoordinate()
  n31 = Multiply(A=Multiply(
    A=@n3.Open
    B=MaterialFunctionCall!(Time=@n3.Open) {materialFunction=/Game/Effects/Fort_Effects/Materials/Geode/_Functions/MF_SineVibration}.Result
   )) {constB=0.05}
  StaticSwitchParameter(
   True=MaterialFunctionCall!(Mask=ComponentMask(None=@n30) {g=true}, Parameter=Add(A=@n3.Open, B=@n31)) {materialFunction=/Game/Effects/Fort_Effects/Materials/Geode/_Functions/MF_Geode_Crack_WPO}.Result
   False=Constant()
  ) {defaultValue=true, parameterName=UseWPO}
 ]
}
