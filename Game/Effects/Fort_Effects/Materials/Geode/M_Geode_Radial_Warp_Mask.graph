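# Masked unlit material for mesh particles that draws a glowing, cracked geode radial-warp effect driven by the "Open" dynamic parameter.
# Bump-offset crack and mask layers at several depths fake interior parallax, combined with cubemap, panning smoke, gradient and view-dependent sphere-mask glow for HDR emissive.
# Open slides crack UVs, animates a smoothstep crack mask (NewCloseAnim switch) and, with sine vibration, feeds optional crack world position offset.
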Material /Game/Effects/Fort_Effects/Materials/Geode/M_Geode_Radial_Warp_Mask
graph {
 properties {
  bUsedWithMeshParticles=true
  bUsedWithNiagaraMeshParticles=true
  blendMode=BLEND_Masked
  pixelDepthOffsetMode=PDOM_Legacy
  previewMesh=/Game/Effects/Fort_Effects/Mesh/Geode/SM_Geode_Radial_Warp
  shadingModel=MSM_Unlit
  usedShadingModels=Unlit
 }
 nodes [
  Constant() {r=1}
  MaterialFunctionCall("FadeSensitivity (optional)"=Constant() {r=1}) {materialFunction=&FadeBasedOnViewAngle}
  n1 = MaterialFunctionCall!(UV add) {materialFunction=/Game/Athena/Effects/Materials/Geode/Functions/MF_Geode_Cubemap}
  n2 = Constant3Vector() {constant=<rgba 0.4835 0.6168 0.8698 0>}
  n3 = Multiply(A=MaterialFunctionCall!(Use Alternate UVs, Alternate UVs) {materialFunction=/Game/Athena/Effects/Materials/Geode/Functions/MF_Geode_Crack_Colors}.Crack Color) {constB=2}
  n4 = VectorParameter() {parameterName=Tile, defaultValue=<rgba 1 1 1 1>}
  n5 = ScalarParameter() {defaultValue=-0.73, parameterName=OpenMin}
  n6 = ScalarParameter() {defaultValue=-0.45, parameterName=OpenMax}
  n7 = DynamicParameter() {paramNames [Param1,Param2,Param3,Open]}
  n8 = Add(
   A=Multiply(A=TextureCoordinate() {uTiling=2, vTiling=1.25}, B=AppendVector(A=@n4.R, B=@n4.R))
   B=AppendVector(A=Constant(), B=LinearInterpolate(A=@n5, B=@n6, Alpha=@n7.Open) {constA=-0.73, constB=-0.45})
  )
  n9 = Add(A=@n8, B=Constant2Vector() {g=-0.001})
  n10 = Multiply(A=NamedRerouteUsage(), B=@n9)
  n11 = TextureCoordinate()
  n12 = LinearInterpolate(Alpha=Power(Base=ComponentMask(None=@n11) {g=true}) {constExponent=12}) {constA=0.5, constB=1.25}
  n13 = Constant() {r=-1}
  n14 = TextureSampleParameter2D(UVs=BumpOffset(Coordinate=@n9, Height=Multiply(A=@n12, B=@n13))) {parameterName=CrackTexture, texture=/Game/Effects/Fort_Effects/Textures/Tile/T_Glass_Cracks_M}
  n15 = Constant() {r=-2.5}
  n16 = TextureSampleParameter2D(UVs=BumpOffset(Coordinate=@n9, Height=Multiply(A=@n12, B=@n15))) {parameterName=CrackTexture, texture=&T_Glass_Cracks_M}
  n17 = BumpOffset(Coordinate=@n10, Height=Multiply(A=@n12, B=Constant() {r=-4}))
  n18 = Clamp(None=Multiply(
    A=Add(A=@n14.G, B=@n16.G)
    B=TextureSampleParameter2D(UVs=@n17) {parameterName=CrackMask, texture=/Game/Effects/Fort_Effects/Textures/Tile/T_Glass_Cracks_Mask}.B
   )) {maxDefault=2}
  n19 = ScalarParameter() {defaultValue=2, parameterName=InteriorCrackBrightness}
  n20 = OneMinus(None=VertexColor().G)
  n21 = Multiply(
   A=Add(
    A=TextureSampleParameter2D(UVs=@n10) {parameterName=CrackMask, texture=&T_Glass_Cracks_Mask}.B
    B=Multiply(A=@n18, B=@n19) {constB=20}
   )
   B=@n20
  )
  n22 = VectorParameter() {parameterName=Glow, defaultValue=<rgba 0.4644 0.8795 1 0>}
  n23 = Multiply(A=TextureSample(UVs=Panner(Coordinate=@n17) {speedX=0.15, speedY=0.2}) {texture=&T_Inky_Smoke_Tile_Chromatic}.RGB) {constB=0.3}
  n24 = CameraVectorWS()
  n25 = VertexNormalWS()
  n26 = ScalarParameter() {defaultValue=6000, parameterName=InnerGlowRadius}
  n27 = ScalarParameter() {defaultValue=1, parameterName=InnerGlowDistance}
  n28 = SphereMask(
   A=Subtract(
    A=WorldPosition().XYZ
    B=MaterialFunctionCall() {materialFunction=&ObjectPivotPoint}.Mesh Particle Pivot Location
   )
   B=Multiply(A=Subtract(A=@n24, B=@n25), B=Multiply(A=@n26, B=@n27) {constB=1.5}) {constB=100}
   Radius=@n26
  ) {attenuationRadius=100, hardnessPercent=0}
  n29 = Add(A=Multiply(A=@n20, B=ComponentMask(None=@n23) {g=true, r=true}), B=@n28)
  n30 = Add(
   A=Multiply(A=@n23, B=@n28)
   B=TextureSample(UVs=Clamp(None=@n29)) {texture=/Game/Effects/Fort_Effects/Textures/Grad/T_Halo_Corona_Gradient}.RGB
  )
  n31 = Add(
   A=Add(A=Multiply(A=@n1.Result, B=@n2) {constB=0.2}, B=Multiply(A=@n3, B=@n21) {constB=2})
   B=Add(A=Multiply(A=@n22.RGB) {constB=0.15}, B=Multiply(A=@n22.RGB, B=@n30))
  )
  MaterialFunctionCall(Emissive=MaterialFunctionCall(LDR Input=Clamp(None=@n31), HDR Tint=@n31) {materialFunction=&DeriveHDRfromLDR}.RemappedRange) {materialFunction=&MF_FX_EmissiveAdjustments}
  MaterialFunctionCall(Speed=Constant() {r=0.5}) {materialFunction=&DebugTimeSine}
  NamedRerouteDeclaration(None=AppendVector(A=@n4.B, B=@n4.B)) {name=CrackMaskTileMult, nodeColor=<rgba 0 1 0.9083 1>}
  n32 = LinearInterpolate(Alpha=@n7.Open) {constA=-0.1, constB=0.5}
  n33 = TextureSampleParameter2D(UVs=Multiply(A=NamedRerouteUsage(), B=@n8)) {parameterName=CrackMask, texture=&T_Glass_Cracks_Mask}
  n34 = Add(
   A=SmoothStep(
    Min=@n32
    Max=Add(A=@n32) {constB=0.1}
    Value=TextureSample(UVs=@n8) {samplerSource=SSM_Wrap_WorldGroupSettings, texture=&T_Glass_Cracks_M}.R
   )
   B=@n33.B
  )
  OneMinus(None=StaticSwitchParameter(True=Saturate(None=@n34), False=@n33.B) {defaultValue=true, parameterName=NewCloseAnim})
  n35 = ComponentMask(None=TextureCoordinate()) {g=true}
  n36 = MaterialFunctionCall!(Time=@n7.Open) {materialFunction=/Game/Effects/Fort_Effects/Materials/Geode/_Functions/MF_SineVibration}
  n37 = Add(A=@n7.Open, B=Multiply(A=Multiply(A=@n7.Open, B=@n36.Result)) {constB=0.05})
  Reroute(None=StaticSwitchParameter(
    True=MaterialFunctionCall!(Mask=@n35, Parameter=@n37) {materialFunction=/Game/Effects/Fort_Effects/Materials/Geode/_Functions/MF_Geode_Crack_WPO}.Result
    False=Constant()
   ) {defaultValue=true, parameterName=UseWPO})
 ]
}
