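# Translucent unlit particle material for a sharp cross flare sprite in Geode effects, usable on Cascade and Niagara sprites.
# Emissive adds a tight center glow, dropped on low quality, to particle color tinted by Geode crack colors.
# Opacity combines flare texture, particle alpha, Geode fade, camera/depth fades and a layered-sine flicker driven by dynamic parameters.
# The "GPU Sprites?" switch substitutes random hue-shifted particle color for the textured emissive and disables flicker.
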Material /Game/Effects/Fort_Effects/Materials/Geode/M_Geode_SharpCross
graph {
 properties {
  bUsedWithNiagaraSprites=true
  bUsedWithParticleSprites=true
  blendMode=BLEND_Translucent
  pixelDepthOffsetMode=PDOM_Legacy
  shadingModel=MSM_Unlit
  usedShadingModels=Unlit
 }
 nodes [
  n1 = ParticleColor()
  n2 = SphereMask(A=TextureCoordinate(), B=Constant() {r=0.5}) {attenuationRadius=0.08, hardnessPercent=0}
  n3 = Multiply(A=Multiply(A=LinearInterpolate(A=@n1.RGB) {constAlpha=0.2}, B=Power(Base=@n2) {constExponent=8})) {constB=2}
  n4 = Constant()
  n5 = MaterialFunctionCall(Hue Shift Percentage=ParticleRandom(), Texture=@n1.RGB) {materialFunction=&HueShift}
  n6 = MaterialFunctionCall!(Use Alternate UVs, Alternate UVs) {materialFunction=/Game/Athena/Effects/Materials/Geode/Functions/MF_Geode_Crack_Colors}
  n7 = TextureSample() {samplerType=SAMPLERTYPE_Grayscale, texture=/Game/Effects/Fort_Effects/Textures/Flares/T_Cross_Flare}
  n8 = Multiply(A=@n1.RGB, B=LinearInterpolate(A=@n6.Cloud Color, B=@n6.Cloud Color, Alpha=@n7.R))
  MaterialFunctionCall(Emissive=Add(A=QualitySwitch(Default=@n3, Low=@n4), B=StaticSwitchParameter(True=@n5.Result, False=@n8) {parameterName=GPU Sprites?})) {materialFunction=&MF_FX_EmissiveAdjustments}
  n9 = Multiply(A=@n1.A, B=@n7.R)
  n10 = QualitySwitch(
   Default=DepthFade(
    Opacity=MaterialFunctionCall() {materialFunction=&Near_Camera_Fade}.Result
    FadeDistance=ScalarParameter() {defaultValue=30, parameterName=DepthFadeDistance}
   )
   Low=Constant() {r=1}
  )
  n11 = DynamicParameter() {paramNames [FlickerFade,FlickerSpeed,Param3,Param4]}
  n12 = Multiply(A=Time(), B=@n11.FlickerSpeed)
  n13 = Multiply(A=Sine(None=@n12) {period=0.155}, B=Sine(None=@n12) {period=0.2})
  n14 = Sine(None=@n12) {period=0.43}
  n15 = Clamp(None=Multiply(A=Sine(None=@n12) {period=0.075}, B=Multiply(A=@n13, B=@n14))) {maxDefault=3}
  Multiply(
   A=Multiply(
    A=MaterialFunctionCall!(In) {materialFunction=/Game/Effects/Fort_Effects/Materials/Geode/_Functions/MF_Geode_Fade}.Opacity_XAxis Particle Sprites
    B=Multiply(A=@n9, B=@n10)
   )
   B=StaticSwitchParameter(True=Constant() {r=1}, False=LinearInterpolate(A=@n15, Alpha=@n11.FlickerFade) {constA=1}) {parameterName=GPU Sprites?}
  )
 ]
}
