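# Translucent, unlit, two-sided mesh-particle material for small refractive geode fragments.
# Tiled glass-crack textures drive index-of-refraction distortion, masked by particle alpha, UV gradient and vertex color, and faded by view angle.
# Color sums geode crack and cloud colors, fake directional lighting from the crack normal, and a vertex-blue-boosted cubemap, tinted by particle color.
# A crack world-position-offset function takes a vertex-green mask and the dynamic "Open" parameter.
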Material /Game/Effects/Fort_Effects/Materials/Geode/M_Geode_Small_Refract_Frags
graph {
 properties {
  bUsedWithMeshParticles=true
  blendMode=BLEND_Translucent
  pixelDepthOffsetMode=PDOM_Legacy
  refractionMethod=RM_IndexOfRefraction
  shadingModel=MSM_Unlit
  twoSided=true
  usedShadingModels=Unlit
 }
 nodes [
  n1 = TextureCoordinate() {uTiling=3}
  n2 = TextureSample(UVs=@n1) {texture=/Game/Effects/Fort_Effects/Textures/Tile/T_Glass_Cracks_M}
  n3 = ParticleColor()
  n4 = ComponentMask(None=TextureCoordinate()) {g=true}
  n5 = VertexColor()
  n6 = OneMinus(None=@n5.B)
  n7 = Multiply(A=@n3.A, B=Multiply(A=@n4, B=Multiply(A=@n5.R, B=@n6)))
  n8 = Constant() {r=0.5}
  LinearInterpolate(
   B=LinearInterpolate(B=ScalarParameter() {defaultValue=1.05, parameterName=RefractionAmount}, Alpha=@n2.R) {constA=0.95}
   Alpha=Multiply(
    A=@n7
    B=MaterialFunctionCall("FadeSensitivity (optional)"=@n8) {materialFunction=&FadeBasedOnViewAngle}.Result
   )
  ) {constA=1}
  MaterialFunctionCall() {materialFunction=&DebugTimeSine}
  MaterialFunctionCall!(Mask=Power(Base=VertexColor().G), Parameter=DynamicParameter() {defaultValue=<rgba 1 1 1 0>, paramNames [Param1,Param2,Pan,Open]}.Open) {materialFunction=/Game/Effects/Fort_Effects/Materials/Geode/_Functions/MF_Geode_Crack_WPO}
  Multiply(A=@n7) {constB=0.1}
  n9 = MaterialFunctionCall!(Use Alternate UVs, Alternate UVs) {materialFunction=/Game/Athena/Effects/Materials/Geode/Functions/MF_Geode_Crack_Colors}
  n10 = Clamp(None=DotProduct(
    A=TextureSample(UVs=@n1) {samplerType=SAMPLERTYPE_Normal, texture=/Game/Effects/Fort_Effects/Textures/Tile/T_Glass_Cracks_N}.RGB
    B=Constant3Vector() {constant=<rgba 0 1 0 0>}
   ))
  n11 = Multiply(A=@n9.Crack Color) {constB=5}
  n12 = Multiply(A=@n11, B=Power(Base=@n4))
  n13 = Multiply(A=Multiply(A=@n2.G) {constB=2}, B=Add(A=@n9.Cloud Color, B=Multiply(A=@n10, B=@n12) {constB=2}))
  n14 = Multiply(A=@n4, B=@n5.B)
  n15 = Multiply(A=@n11, B=Clamp(None=Power(Base=@n14) {constExponent=10}) {maxDefault=2})
  n16 = MaterialFunctionCall!(UV add) {materialFunction=/Game/Athena/Effects/Materials/Geode/Functions/MF_Geode_Cubemap}
  n17 = LinearInterpolate(Alpha=Power(Base=@n5.B) {constExponent=3}) {constA=1, constB=20}
  Multiply(A=@n3.RGB, B=Add(A=Add(A=@n13, B=@n15), B=Multiply(A=@n16.Result, B=@n17)))
 ]
}
