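# Translucent unlit test material for Geode effects, drawing glowing, cloudy electric cracks on spline mesh particles.
# Dynamic parameters drive width, panning, crack growth and opening, feeding SplineThicken with near-camera scaling plus optional vibrating crack offset.
# Crack texture, distorted by two panning turbulence layers, combines with sphere and width masks to build emissive glow and depth-faded opacity.
# Static switches add length-wise color gradients, center masking and animated or world-sphere reveals.
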Material /Game/Effects/Fort_Effects/Materials/Geode/M_Geode_Test_Spline
graph {
 properties {
  bUsedWithMeshParticles=true
  bUsedWithNiagaraMeshParticles=true
  blendMode=BLEND_Translucent
  pixelDepthOffsetMode=PDOM_Legacy
  shadingModel=MSM_Unlit
  twoSided=true
  usedShadingModels=Unlit
 }
 nodes [
  n1 = DynamicParameter() {defaultValue=<rgba 1 1 1 0>, paramNames [Crack, SplineWidthMult, Pan, Open]}
  n2 = ScalarParameter() {defaultValue=4000, parameterName=NearCameraWidthDisrance}
  n3 = Multiply(
   A=Multiply(A=@n1.SplineWidthMult, B=ScalarParameter() {defaultValue=500, parameterName=Width}) {constA=1}
   B=ShadingPathSwitch(
    Default=MaterialFunctionCall(Distance=@n2) {materialFunction=&Near_Camera_Fade}.Result
    Mobile=ScalarParameter() {defaultValue=1, parameterName=NearCameraWidth_Mobile}
   ) {desc=Migrated from FeatureLevelSwitch}
  )
  n4 = ScalarParameter() {defaultValue=1, parameterName=WidthBaseMult}
  n5 = Power(Base=ComponentMask(None=TextureCoordinate()) {g=true})
  n6 = MaterialFunctionCall!(Time=@n1.Open) {materialFunction=/Game/Effects/Fort_Effects/Materials/Geode/_Functions/MF_SineVibration}
  n7 = Add(A=@n1.Open, B=Multiply(A=Multiply(A=@n1.Open, B=@n6.Result)) {constB=0.05})
  Add(
   A=MaterialFunctionCall(WidthBase=Multiply(A=@n3, B=@n4), WidthTip=Multiply(A=@n3)) {materialFunction=&SplineThicken}.WorldPositionOffset
   B=StaticSwitchParameter(
    True=MaterialFunctionCall!(Mask=@n5, Parameter=@n7) {materialFunction=/Game/Effects/Fort_Effects/Materials/Geode/_Functions/MF_Geode_Crack_WPO}.Result
    False=Constant()
   ) {defaultValue=true, parameterName=UseWPO}
  ) {constB=0}
  Constant() {r=500}
  Constant() {r=1}
  Constant() {r=1}
  n8 = ParticleColor()
  n9 = VertexColor()
  n10 = TextureCoordinate() {coordinateIndex=1}
  n11 = MaterialFunctionCall(Float2=@n10) {materialFunction=&BreakOutFloat2Components}
  n12 = MaterialFunctionCall(X=@n11.R, Y=OneMinus(None=@n11.G)) {materialFunction=&MakeFloat2}
  n13 = AppendVector(A=Constant() {r=1}, B=ScalarParameter() {defaultValue=1, parameterName=TileMult})
  n14 = LinearInterpolate(B=@n1.Crack, Alpha=ScalarParameter() {defaultValue=1, parameterName=CrackSize})
  n15 = Multiply(A=@n1.Pan, B=@n14)
  n16 = Add(
   A=Multiply(A=StaticSwitchParameter(True=@n12.Result, False=@n10) {parameterName=InvertDirection}, B=@n13)
   B=AppendVector(A=Constant(), B=StaticSwitchParameter(True=@n15, False=@n1.Pan) {parameterName=Crack spread affects panning})
  )
  n17 = TextureSample(UVs=@n16) {texture=/Game/Effects/Fort_Effects/Textures/Electric/T_Vert_Electric_Cracks}
  n18 = LinearInterpolate(Alpha=@n17.B) {constA=1, constB=-1}
  n19 = TextureCoordinate()
  n20 = ComponentMask(None=TextureCoordinate()) {g=true}
  n21 = VectorParameter() {parameterName=SphereMaskCenter, defaultValue=<rgba 0.5 0.5 0 1>}
  n22 = SphereMask(
   A=TextureCoordinate() {coordinateIndex=2}
   B=MaterialFunctionCall(X=@n21.R, Y=@n21.R) {materialFunction=&MakeFloat2}.Result
   Radius=LinearInterpolate(Alpha=@n14) {constB=2}
   Hardness=LinearInterpolate(Alpha=@n14) {constA=0.7}
  ) {hardnessPercent=0}
  n23 = Multiply(A=Clamp(None=Power(Base=@n20) {constExponent=0.6}) {minDefault=0.1}, B=@n22)
  n24 = AppendVector(A=@n23, B=Constant() {r=1})
  n25 = SphereMask(
   A=ComponentMask(None=MaterialFunctionCall(UVs=@n19, Texture Scale=@n24) {materialFunction=&ScaleUVsByCenter}.UVs) {r=true}
   B=Constant() {r=0.5}
   Radius=ScalarParameter() {defaultValue=0.5, parameterName=AlphaMask_Width}
  ) {attenuationRadius=0.2, hardnessPercent=0}
  n26 = TextureCoordinate() {coordinateIndex=1}
  n27 = MaterialFunctionCall(Float2=@n26) {materialFunction=&BreakOutFloat2Components}
  n28 = StaticSwitchParameter(
   True=MaterialFunctionCall(X=@n27.R, Y=OneMinus(None=@n27.G)) {materialFunction=&MakeFloat2}.Result
   False=@n26
  ) {parameterName=InvertDirection}
  n29 = ScalarParameter() {defaultValue=1, parameterName=Turbulence_Tile_U}
  n30 = ScalarParameter() {defaultValue=1, parameterName=Turbulence_Tile_V}
  n31 = Add(
   A=Multiply(A=Multiply(A=@n18, B=@n25) {constA=1}, B=Constant2Vector() {r=-0.06}) {constB=0.2}
   B=Multiply(A=Multiply(A=@n28, B=@n13), B=AppendVector(A=@n29, B=@n30))
  )
  n32 = Multiply(A=Time(), B=ScalarParameter() {defaultValue=0.1, parameterName=PanSpeed})
  n33 = Panner(Coordinate=@n31, Time=@n32) {speedX=0.1, speedY=0.05}
  n34 = ParticleRandom()
  n35 = ScalarParameter() {parameterName=Use_ParticleRandom_NoiseOffset}
  n36 = LinearInterpolate(A=@n33, B=Add(A=@n34, B=@n33), Alpha=@n35)
  n37 = TextureSampleParameter2D(UVs=@n36) {parameterName=Turbulence_Normal, samplerType=SAMPLERTYPE_Normal, texture=/Game/Effects/Fort_Effects/Textures/Plasma/T_Turbulence_Seamless_N}
  n38 = Panner(Coordinate=@n31, Time=@n32) {speedX=-0.1, speedY=0.05}
  n39 = LinearInterpolate(A=@n38, B=Add(A=@n34, B=@n38), Alpha=@n35)
  n40 = TextureSampleParameter2D(UVs=@n39) {parameterName=Turbulence_Normal, samplerType=SAMPLERTYPE_Normal, texture=&T_Turbulence_Seamless_N}
  n41 = Constant2Vector()
  n42 = Clamp(
   None=DotProduct(A=LinearInterpolate(A=@n37.RGB, B=@n40.RGB, Alpha=@n17.B), B=AppendVector(A=@n18, B=@n41))
   Max=ScalarParameter() {defaultValue=1, parameterName=Turbulence_Strength_Normal}
  )
  n43 = QualitySwitch(Default=Multiply(A=@n42, B=Constant2Vector() {r=0.011}), Low=Constant())
  n44 = TextureSample(UVs=Add(A=Multiply(A=@n43, B=@n18), B=@n16)) {texture=&T_Vert_Electric_Cracks}
  n45 = LinearInterpolate(
   A=TextureSampleParameter2D(UVs=@n36) {parameterName=Turbulence, texture=/Game/Global/Textures/Default/Caustics/T_Turbulence_Seamless}.R
   B=TextureSampleParameter2D(UVs=@n39) {parameterName=Turbulence, texture=&T_Turbulence_Seamless}.R
   Alpha=@n17.B
  )
  n46 = Multiply(A=@n45, B=ScalarParameter() {defaultValue=2, parameterName=Turbulence_Strength}) {constB=2}
  n47 = Add(
   A=Multiply(A=@n44.RGB, B=Constant3Vector() {constant=<rgba 1 1 0 0>})
   B=Multiply(A=Add(A=@n46) {constB=0.15}, B=Power(Base=@n25)) {constA=1}
  )
  n48 = Multiply(A=@n8.A, B=Multiply(A=Multiply(A=@n9.G, B=@n47), B=@n22))
  n49 = ShadingPathSwitch(
   Default=DepthFade(Opacity=@n48, FadeDistance=ScalarParameter() {defaultValue=10000, parameterName=DepthFadeDIst}) {fadeDistanceDefault=500}
   Mobile=@n48
  ) {desc=Migrated from FeatureLevelSwitch}
  LinearInterpolate(B=ScalarParameter() {defaultValue=1.05, parameterName=Refraction}, Alpha=Power(Base=ComponentMask(None=@n49) {g=true}) {constExponent=3}) {constA=1}
  MaterialFunctionCall() {materialFunction=&DebugTimeSine}
  n50 = ComponentMask(None=TextureCoordinate()) {g=true}
  n51 = ScalarParameter() {group=LerpColorAlongSpline, parameterName=LerpColorMin, sortPriority=3}
  n52 = ScalarParameter() {defaultValue=1, group=LerpColorAlongSpline, parameterName=LerpColorMax, sortPriority=4}
  n53 = LinearInterpolate(
   A=VectorParameter() {group=LerpColorAlongSpline, parameterName=LerpColorA, sortPriority=1, defaultValue=<rgba 1 1 1 1>}.RGB
   B=VectorParameter() {group=LerpColorAlongSpline, parameterName=LerpColorB, sortPriority=2, defaultValue=<rgba 1 1 1 1>}.RGB
   Alpha=Saturate(None=MaterialFunctionCall(Input=@n50, Input Low=@n51, Input High=@n52) {materialFunction=&RemapValueRange}.Result)
  )
  n54 = TextureCoordinate() {coordinateIndex=1}
  n55 = TextureSample(UVs=Add(A=Rotator(Coordinate=@n54) {speed=0.5}, B=Multiply(A=@n45) {constB=0.6})) {texture=/Game/Global/Textures/Default/Masks_and_Gradients/T_Grad_Rainbow}
  n56 = Add(
   A=QualitySwitch(Default=Multiply(A=@n55.RGB) {constB=0.2}, Low=Constant())
   B=MaterialFunctionCall!(Use Alternate UVs, Alternate UVs) {materialFunction=/Game/Athena/Effects/Materials/Geode/Functions/MF_Geode_Crack_Colors}.Cloud Color
  )
  n57 = Multiply(A=@n25, B=ScalarParameter() {defaultValue=1.5, parameterName=CloudGlowMult}) {constB=2}
  n58 = QualitySwitch(Default=Power(Base=Multiply(A=@n42, B=@n57)) {constExponent=3}, Low=Constant())
  n59 = Multiply(A=Power(Base=@n25) {constExponent=5}, B=ScalarParameter() {defaultValue=0.2, parameterName=CloudGlowAdd}) {constB=0.15}
  n60 = VectorParameter() {parameterName=CrackGlowMult, defaultValue=<rgba 0.3893 0.3473 0.5052 1>}
  n61 = MaterialFunctionCall() {materialFunction=&MF_Geode_Crack_Colors}
  n62 = Add(A=@n44.R, B=@n44.R)
  n63 = LinearInterpolate(Alpha=@n23) {constA=0.8, constB=2}
  n64 = Add(A=Multiply(A=Add(A=@n58, B=@n59), B=Multiply(A=@n60.RGB, B=@n61.Crack Color)), B=Multiply(A=Power(Base=@n62, Exp=@n63), B=@n61.Crack Color))
  n65 = LinearInterpolate(A=VectorParameter() {parameterName=CrackGlowAlongLengthMult, defaultValue=<rgba 0.2133 0.1829 0.2344 1>}.RGB, Alpha=@n23)
  n66 = Multiply(A=@n8.RGB, B=Add(A=Multiply(A=@n56, B=@n46), B=Multiply(A=@n64, B=@n65)))
  MaterialFunctionCall(Emissive=StaticSwitchParameter(True=Multiply(A=@n53, B=@n66), False=@n66) {group=LerpColorAlongSpline, parameterName=LerpColorAlongSpline, sortPriority=0}) {materialFunction=&MF_FX_EmissiveAdjustments}
  ScalarParameter() {defaultValue=400, parameterName=test}
  n67 = ComponentMask(None=TextureCoordinate()) {g=true}
  n68 = Multiply(
   A=Power(Base=OneMinus(None=@n67), Exp=ScalarParameter() {defaultValue=4, parameterName=CenterMask_Power}) {constExponent=4}
   B=ScalarParameter() {defaultValue=40, parameterName=CenterMask_Multiply}
  ) {constB=40}
  n69 = StaticSwitchParameter(True=Multiply(A=Saturate(None=@n68), B=@n49), False=@n49) {parameterName=CenterMask}
  n70 = DynamicParameter() {parameterIndex=1, defaultValue=<rgba 1 1 1 0>, paramNames ["", "", "", CrackMaskAnim]}
  n71 = OneMinus(None=ComponentMask(None=TextureCoordinate()) {g=true})
  n72 = Saturate(None=Divide(A=Add(A=@n70.CrackMaskAnim, B=@n71), B=ScalarParameter() {defaultValue=1, parameterName=CrackMaskAnimHardness}))
  n73 = DynamicParameter() {parameterIndex=3, defaultValue=<rgba 0 1 1 0>, paramNames [SphereMaskOpenSize, Param2, Param3, Param4]}
  n74 = OneMinus(None=SphereMask(A=WorldPosition().XYZ, B=ActorPositionWS(), Radius=@n73.SphereMaskOpenSize) {hardnessPercent=95})
  n75 = ScalarParameter() {parameterName=UseSphereMaskOpening}
  StaticSwitchParameter(True=Multiply(A=@n69, B=LinearInterpolate(A=@n72, B=@n74, Alpha=@n75)), False=@n69) {parameterName=CrackMaskAnim}
  VertexColor()
 ]
}
