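# Material function for the Geode rift effect that computes a world-space offset to twist and shrink geometry around a CenterPoint parameter.
# RotateAboutAxis twists positions around Z by Corruption times TwistAmount, weighted by two powered sphere masks remapped to start slightly negative.
# A pull toward the center, scaled by ShrinkMult and Corruption, grows with the inverted ObjectRadius mask and is added to the twist.
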MaterialFunction /Game/Effects/Fort_Effects/Materials/Geode/_Functions/MF_Rift_Twist
graph {nodes [
  n1 = Constant3Vector() {constant=<rgba 0 0 1 0>}
  n2 = Multiply(A=ScalarParameter() {parameterName=Corruption}, B=ScalarParameter() {defaultValue=0.1, parameterName=TwistAmount}) {constB=0.1}
  n3 = WorldPosition()
  n4 = VectorParameter() {parameterName=CenterPoint, defaultValue=<rgba 8385 3888 105390 0>}
  n5 = WorldPosition()
  n6 = ScalarParameter() {defaultValue=90000, parameterName=ObjectRadius}
  n7 = SphereMask(A=@n5.XYZ, B=@n4.RGB, Radius=Multiply(A=@n6) {constB=0.25}) {attenuationRadius=20000, hardnessPercent=0}
  n8 = SphereMask(A=@n4.RGB, B=@n5.XYZ, Radius=@n6) {hardnessPercent=0}
  n9 = Add(
   A=Multiply(A=Power(Base=@n7) {constExponent=5}, B=ScalarParameter() {defaultValue=0.5}) {constB=0.1}
   B=Power(Base=@n8, Exp=ScalarParameter() {defaultValue=3, parameterName=TwistPower})
  )
  n10 = Multiply(A=Subtract(A=@n4.RGB, B=WorldPosition().XYZ), B=ScalarParameter() {defaultValue=0.8, parameterName=ShrinkMult})
  n11 = ScalarParameter() {parameterName=Corruption}
  n12 = Add(
   A=Multiply(
    A=RotateAboutAxis(NormalizedRotationAxis=@n1, RotationAngle=@n2, PivotPoint=@n4.RGB, Position=@n3.XYZ)
    B=LinearInterpolate(Alpha=@n9) {constA=-0.25}
   )
   B=Multiply(A=OneMinus(None=@n8), B=Multiply(A=@n10, B=@n11))
  )
  FunctionOutput(None=@n12)
  MaterialFunctionCall() {materialFunction=&DebugTimeSine}
  Multiply(A=@n12, B=Constant3Vector() {constant=<rgba 1 1 0 0>})
  OneMinus(None=ComponentMask(None=TextureCoordinate()) {g=true})
 ]}
