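# Translucent unlit mesh-particle material for Geode rift spline cracks, a test healing variant tinted by HealColor.
# Dynamic Crack and Width parameters, vertex color and noise build a growth mask driving glow, opacity and SplineThicken width, with MF_Rift_Twist added to the offset.
# Panned turbulence clouds distort a rainbow gradient, while their normals light the glow and offset crack UVs.
# Low quality and mobile drop costlier terms; a RiftBeacon switch scales opacity by MPC_Nexus RiftOpacity.
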Material /Game/Effects/Fort_Effects/Materials/Geode/M_Geode_Test_Spline_Healing
graph {
 properties {
  bUsedWithMeshParticles=true
  blendMode=BLEND_Translucent
  pixelDepthOffsetMode=PDOM_Legacy
  shadingModel=MSM_Unlit
  twoSided=true
  usedShadingModels=Unlit
 }
 nodes [
  n1 = ComponentMask(None=TextureCoordinate()) {g=true}
  n2 = DynamicParameter() {paramNames [Crack,Width,Param3,Param4]}
  n3 = ScalarParameter() {defaultValue=1, parameterName=Opacity}
  n4 = ScalarParameter() {parameterName=MinOpacity}
  n5 = LinearInterpolate(A=ScalarParameter() {defaultValue=0.4, parameterName=OpacityMin}, Alpha=Multiply(A=@n2.Crack, B=Clamp(None=@n3, Min=@n4))) {constA=0.4, constB=1.1}
  n6 = VertexColor()
  n7 = Add(A=Multiply(A=@n5, B=@n6.R) {constB=0.5}, B=Multiply(A=@n6.R) {constB=0.75}) {constB=0.5}
  n8 = Add(
   A=TextureSample(UVs=TextureCoordinate() {coordinateIndex=2, uTiling=1.5, vTiling=1.5}) {samplerType=SAMPLERTYPE_Grayscale, texture=&LowResBlurredNoise}.G
   B=@n5
  )
  n9 = Clamp(None=Multiply(A=Power(Base=@n1) {constExponent=0.6}, B=LinearInterpolate(B=@n7, Alpha=@n8) {constA=-0.6, constB=0.5}))
  n10 = LinearInterpolate(Alpha=Multiply(A=@n9, B=VertexColor().R)) {constA=0.3}
  n11 = Multiply(A=@n2.Width, B=ScalarParameter() {defaultValue=8000, parameterName=Width})
  n12 = MaterialFunctionCall(Distance=ScalarParameter() {defaultValue=4000, parameterName=NearCameraWidthDisrance}) {materialFunction=&Near_Camera_Fade}
  n13 = Constant() {r=1}
  n14 = Multiply(A=Multiply(A=Multiply(A=@n10, B=@n11), B=ShadingPathSwitch(Default=@n12.Result, Mobile=@n13) {desc=Migrated from FeatureLevelSwitch}))
  Add(
   A=MaterialFunctionCall(WidthBase=@n14, WidthTip=@n14) {materialFunction=&SplineThicken}.WorldPositionOffset
   B=MaterialFunctionCall() {materialFunction=/Game/Effects/Fort_Effects/Materials/Geode/_Functions/MF_Rift_Twist}.Result
  )
  MaterialFunctionCall() {materialFunction=&DebugTimeSine}
  n15 = TextureCoordinate() {coordinateIndex=1, vTiling=0.8}
  n16 = TextureSample(UVs=@n15) {texture=/Game/Effects/Fort_Effects/Textures/Electric/T_Vert_Electric_Cracks}
  n17 = LinearInterpolate(Alpha=@n16.B) {constA=1, constB=-1}
  n18 = SphereMask(A=ComponentMask(None=TextureCoordinate()) {r=true}, B=Constant() {r=0.5}) {attenuationRadius=0.5, hardnessPercent=0}
  n19 = Constant() {r=1}
  n20 = ScalarParameter() {defaultValue=1.7, parameterName=YTile}
  n21 = Add(
   A=Multiply(A=Multiply(A=@n17, B=@n18) {constA=1}, B=Constant2Vector() {r=-0.06}) {constB=0.2}
   B=Multiply(A=TextureCoordinate() {coordinateIndex=1, uTiling=1.5}, B=AppendVector(A=@n19, B=@n20))
  )
  n22 = Multiply(A=Time(), B=ScalarParameter() {defaultValue=0.1, parameterName=PanSpeed})
  n23 = Panner(Coordinate=@n21, Time=@n22) {speedX=0.1, speedY=0.05}
  n24 = Panner(Coordinate=@n21, Time=@n22) {speedX=-0.1, speedY=0.05}
  n25 = LinearInterpolate(
   A=TextureSample(UVs=@n23) {texture=/Game/Global/Textures/Default/Caustics/T_Turbulence_Seamless}.R
   B=TextureSample(UVs=@n24) {texture=&T_Turbulence_Seamless}.R
   Alpha=@n16.B
  )
  n26 = Add(A=Rotator(Coordinate=TextureCoordinate() {coordinateIndex=1}) {speed=0.5}, B=Multiply(A=@n25) {constB=0.6})
  n27 = QualitySwitch(
   Default=Multiply(A=TextureSample(UVs=@n26) {texture=/Game/Global/Textures/Default/Masks_and_Gradients/T_Grad_Rainbow}.RGB) {constB=0.2}
   Low=Constant()
  )
  n28 = ScalarParameter() {parameterName=Corruption}
  n29 = Multiply(A=@n25) {constB=2}
  n30 = Multiply(
   A=Add(
    A=LinearInterpolate(A=@n27, Alpha=@n28) {constA=0.2, constB=0}
    B=MaterialFunctionCall!(Use Alternate UVs, Alternate UVs) {materialFunction=/Game/Athena/Effects/Materials/Geode/Functions/MF_Geode_Crack_Colors}.Cloud Color
   )
   B=@n29
  )
  n31 = TextureSample(UVs=@n23) {samplerType=SAMPLERTYPE_Normal, texture=/Game/Effects/Fort_Effects/Textures/Plasma/T_Turbulence_Seamless_N}
  n32 = TextureSample(UVs=@n24) {samplerType=SAMPLERTYPE_Normal, texture=&T_Turbulence_Seamless_N}
  n33 = Constant2Vector()
  n34 = Clamp(None=DotProduct(A=LinearInterpolate(A=@n31.RGB, B=@n32.RGB, Alpha=@n16.B), B=AppendVector(A=@n17, B=@n33)))
  n35 = MaterialFunctionCall(In=@n9, Contrast=Constant() {r=3}) {materialFunction=&CheapContrast}
  n36 = LinearInterpolate(Alpha=@n35.Result) {constA=0.4}
  n37 = Multiply(A=@n18, B=@n36)
  n38 = Multiply(A=ScalarParameter() {defaultValue=1.5, parameterName=CloudGlowMult}, B=@n37) {constB=2}
  n39 = QualitySwitch(Default=Power(Base=Multiply(A=@n34, B=@n38)) {constExponent=3}, Low=Constant())
  n40 = Multiply(A=Power(Base=@n37) {constExponent=5}, B=ScalarParameter() {defaultValue=0.2, parameterName=CloudGlowAdd}) {constB=0.15}
  n41 = VectorParameter() {parameterName=CrackGlowMult, defaultValue=<rgba 0.2899 0.25 0.4 1>}
  n42 = Multiply(
   A=MaterialFunctionCall() {materialFunction=&MF_Geode_Crack_Colors}.Crack Color
   B=LinearInterpolate(A=VectorParameter() {parameterName=HealColor, defaultValue=<rgba 3 0.2543 2.4825 1>}.RGB, Alpha=@n35.Result) {constA=3}
  )
  n43 = QualitySwitch(Default=Multiply(A=@n34, B=Constant2Vector() {r=0.011}), Low=Constant())
  n44 = TextureSample(UVs=Add(A=Multiply(A=@n43, B=@n17), B=@n15)) {texture=&T_Vert_Electric_Cracks}
  n45 = Multiply(A=Add(A=@n44.R, B=Multiply(A=@n44.G) {constB=0.7}), B=@n36)
  n46 = Add(A=Multiply(A=Add(A=@n39, B=@n40), B=Multiply(A=@n41.RGB, B=@n42)), B=Multiply(A=@n45, B=@n42))
  MaterialFunctionCall(Emissive=Multiply(A=ParticleColor().RGB, B=Add(A=@n30, B=@n46))) {materialFunction=&MF_FX_EmissiveAdjustments}
  n47 = Add(A=@n29, B=@n45)
  n48 = Multiply(A=@n18, B=Clamp(None=@n9) {minDefault=0.3})
  n49 = Multiply(A=VertexColor().G, B=Multiply(A=Add(A=@n47) {constB=0.2}, B=Power(Base=@n48)) {constA=1})
  n50 = ScalarParameter() {defaultValue=1246584, parameterName=DepthFade}
  Multiply(
   A=StaticSwitchParameter(
    True=CollectionParameter() {collection=/Game/Athena/Environments/Nexus/Materials/FXshader/MPC_Nexus, parameterName=RiftOpacity}
    False=Constant() {r=1}
   ) {parameterName=RiftBeacon}
   B=ShadingPathSwitch(Default=DepthFade(Opacity=@n49, FadeDistance=@n50), Mobile=@n49) {desc=Migrated from FeatureLevelSwitch}
  )
 ]
}
