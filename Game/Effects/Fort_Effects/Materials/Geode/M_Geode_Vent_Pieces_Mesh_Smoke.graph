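# Translucent unlit mesh-particle material for panning smoke, apparently for Geode vent piece effects.
# A panning normal map warps the UVs of three smoke texture samples, offset in U and panned at separate speeds, then combined and fully desaturated.
# That noise, tinted by particle color and dimmed by inverted Fresnel, feeds emissive adjustments.
# A sharpened copy, masked toward UV edges, feeds depth fade and inverted-Fresnel branches.
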Material /Game/Effects/Fort_Effects/Materials/Geode/M_Geode_Vent_Pieces_Mesh_Smoke
graph {
 properties {
  bUsedWithMeshParticles=true
  blendMode=BLEND_Translucent
  pixelDepthOffsetMode=PDOM_Legacy
  shadingModel=MSM_Unlit
  translucencyLightingMode=TLM_VolumetricPerVertexNonDirectional
  usedShadingModels=Unlit
 }
 nodes [
  n1 = ScalarParameter() {defaultValue=1, parameterName=BaseTilingU}
  n2 = ScalarParameter() {defaultValue=1, parameterName=BaseTilingV}
  n3 = TextureCoordinate() {coordinateIndex=1}
  n4 = AppendVector(
   A=ScalarParameter() {defaultValue=0.25, parameterName=WarpTexTilingU, sliderMin=0.25}
   B=ScalarParameter() {defaultValue=1, parameterName=WarpTexTilingV}
  )
  n5 = ScalarParameter() {defaultValue=0.02, parameterName=WarpPanSpeedU}
  n6 = ScalarParameter() {defaultValue=-0.1, parameterName=WarpPanSpeedV}
  n7 = TextureSampleParameter2D(UVs=Panner(Coordinate=Multiply(A=@n3, B=@n4), Speed=AppendVector(A=@n5, B=@n6)) {bFractionalPart=true, speedY=0.25}) {parameterName=UVDistTex, samplerType=SAMPLERTYPE_Normal, texture=/Game/Global/Textures/Default/Tileable_Patterns/Cloud_normals}
  n8 = Add(
   A=Multiply(A=TextureCoordinate() {coordinateIndex=1}, B=AppendVector(A=@n1, B=@n2))
   B=Multiply(A=ComponentMask(None=@n7.RGB) {g=true, r=true}, B=ScalarParameter() {defaultValue=0.05, parameterName=DistortAmount}) {constB=0.0125}
  )
  n9 = ScalarParameter() {defaultValue=1, parameterName=ChromaShiftDistance}
  n10 = Multiply(A=Constant2Vector() {r=1}, B=@n9)
  n11 = ScalarParameter() {parameterName=PanSpeedU}
  n12 = ScalarParameter() {defaultValue=-0.22, parameterName=PanSpeedV}
  n13 = TextureSampleParameter2D(UVs=Panner(Coordinate=Add(A=@n8, B=@n10), Speed=AppendVector(A=@n11, B=@n12)) {bFractionalPart=true, speedX=-0.15}) {parameterName=GrayscaleTex, texture=/Game/Global/Textures/Default/Tileable_Patterns/T_Smoke_Puff_Tile_Up}
  n14 = ScalarParameter() {parameterName=PanSpeedU_1}
  n15 = ScalarParameter() {defaultValue=-0.15, parameterName=PanSpeedV_1}
  n16 = TextureSampleParameter2D(UVs=Panner(Coordinate=@n8, Speed=AppendVector(A=@n14, B=@n15)) {bFractionalPart=true, speedX=-0.2}) {parameterName=GrayscaleTex_1, texture=&T_Smoke_Puff_Tile_Up}
  n17 = Multiply(A=@n9, B=Constant2Vector() {r=-1})
  n18 = ScalarParameter() {parameterName=PanSpeedU_2}
  n19 = ScalarParameter() {defaultValue=-0.2, parameterName=PanSpeedV_2}
  n20 = TextureSampleParameter2D(UVs=Panner(Coordinate=Add(A=@n8, B=@n17), Speed=AppendVector(A=@n18, B=@n19)) {bFractionalPart=true, speedX=-0.2}) {parameterName=GrayscaleTex_2, texture=&T_Smoke_Puff_Tile_Up}
  n21 = Desaturation(
   None=MaterialFunctionCall(X=Power(Base=@n13.G) {constExponent=1}, Y=Power(Base=@n16.G) {constExponent=1}, Z=Power(Base=@n20.G) {constExponent=1}) {materialFunction=&MakeFloat3}.Result
   Fraction=Constant() {r=1}
  )
  n22 = Clamp(None=Power(Base=@n21) {constExponent=6})
  Add(A=@n22) {constB=0.1}
  n23 = TextureCoordinate()
  n24 = ComponentMask(None=@n23) {r=true}
  n25 = OneMinus(None=@n24)
  n26 = ComponentMask(None=@n23) {g=true}
  n27 = OneMinus(None=@n26)
  n28 = Multiply(A=@n22, B=Multiply(A=Multiply(A=@n24, B=@n25), B=Multiply(A=@n26, B=@n27)))
  DepthFade(Opacity=@n28, FadeDistance=ScalarParameter() {defaultValue=15, parameterName=FadeDistance})
  n29 = Fresnel() {exponent=0.1}
  n30 = ParticleColor()
  MaterialFunctionCall(Emissive=Multiply(A=OneMinus(None=@n29), B=Multiply(A=@n30.RGB, B=@n21))) {materialFunction=&MF_FX_EmissiveAdjustments}
  n31 = ScalarParameter() {defaultValue=0.1, parameterName=ExponentIn}
  Multiply(A=@n28, B=OneMinus(None=Fresnel(ExponentIn=@n31)))
 ]
}
